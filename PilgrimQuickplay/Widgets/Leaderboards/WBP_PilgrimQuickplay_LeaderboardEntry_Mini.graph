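# Compact list-row widget for Pilgrim Quickplay leaderboards, showing rank, player or band name, accuracy, difficulty and score.
# OnListItemObjectSet passes the list object on to set rank text, alternate zebra shading and play the PlayerEntry highlight for the local player.
# SetEntryData fills the remaining fields through helper functions.
# Friend rank, when positive, replaces global rank; switchers choose globe or friends icon and swap accuracy for a full-combo indicator.
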WidgetBlueprint /PilgrimQuickplay/Widgets/Leaderboards/WBP_PilgrimQuickplay_LeaderboardEntry_Mini
animations [
 PlayerEntry {
  binding_count=4
  length_seconds=0.3
  object=&self:PlayerEntry
  sequence {bindings [
    CommonBorder_Zebra {
     id=#1
     possessed_class=CommonBorder
     tracks [{
       type=MovieSceneColorTrack
       display_name=Brush Color
       sections [{
         type=MovieSceneColorSection
         end_frame=0
         start_frame=0
         channels [
          Color.R {type=MovieSceneScriptingFloatChannel, const [1,1]}
          Color.G {type=MovieSceneScriptingFloatChannel, const [1,1]}
          Color.B {type=MovieSceneScriptingFloatChannel, const [1,1]}
          Color.A {type=MovieSceneScriptingFloatChannel, const [1,1]}
         ]
        }]
      }]
    }
    Text_Score {
     id=#2
     possessed_class=CommonTextBlock
     tracks [{
       type=MovieSceneColorTrack
       display_name=Color and Opacity
       sections [{
         type=MovieSceneColorSection
         end_frame=0
         start_frame=0
         channels [
          Color.R {type=MovieSceneScriptingFloatChannel, const [0.4125,1]}
          Color.G {type=MovieSceneScriptingFloatChannel, const [0.5089,1]}
          Color.B {type=MovieSceneScriptingFloatChannel, const [0.9647,1]}
          Color.A {type=MovieSceneScriptingFloatChannel, const [1,1]}
         ]
        }]
      }]
    }
    Text_Rank {
     id=#3
     possessed_class=CommonTextBlock
     tracks [{
       type=MovieSceneColorTrack
       display_name=Color and Opacity
       sections [{
         type=MovieSceneColorSection
         end_frame=0
         start_frame=0
         channels [
          Color.R {type=MovieSceneScriptingFloatChannel, const [0.4125,1]}
          Color.G {type=MovieSceneScriptingFloatChannel, const [0.5089,1]}
          Color.B {type=MovieSceneScriptingFloatChannel, const [0.9647,1]}
          Color.A {type=MovieSceneScriptingFloatChannel, const [1,1]}
         ]
        }]
      }]
    }
    CurrentUserBG {
     id=#4
     possessed_class=Border
     tracks [{
       type=MovieSceneFloatTrack
       display_name=Render Opacity
       sections [{type=MovieSceneFloatSection, end_frame=6, start_frame=0, channels [None {type=MovieSceneScriptingFloatChannel, keys [[0,0],[6,1]]}]}]
      }]
    }
   ]}
 }
 MainPlayer {
  binding_count=5
  length_seconds=0.3
  object=&self:MainPlayer
  sequence {bindings [
    Text_Score {
     id=#2
     possessed_class=CommonTextBlock
     tracks [{
       type=MovieSceneColorTrack
       display_name=Color and Opacity
       sections [{
         type=MovieSceneColorSection
         end_frame=6
         start_frame=0
         channels [
          Color.R {type=MovieSceneScriptingFloatChannel, keys [[0,0.4125],[6,1]]}
          Color.G {type=MovieSceneScriptingFloatChannel, keys [[0,0.5089],[6,1]]}
          Color.B {type=MovieSceneScriptingFloatChannel, keys [[0,0.9647],[6,1]]}
          Color.A {type=MovieSceneScriptingFloatChannel, const [1,1]}
         ]
        }]
      }]
    }
    FriendIcon {
     id=#5
     possessed_class=Image
     tracks [{
       type=MovieSceneColorTrack
       display_name=Color and Opacity
       sections [{
         type=MovieSceneColorSection
         end_frame=6
         start_frame=0
         channels [
          Color.R {type=MovieSceneScriptingFloatChannel, keys [[0,0.4125],[6,1]]}
          Color.G {type=MovieSceneScriptingFloatChannel, keys [[0,0.5089],[6,1]]}
          Color.B {type=MovieSceneScriptingFloatChannel, keys [[0,0.9647],[6,1]]}
          Color.A {type=MovieSceneScriptingFloatChannel, const [1,2]}
         ]
        }]
      }]
    }
    Text_Accuracy {
     id=#6
     possessed_class=CommonTextBlock
     tracks [{
       type=MovieSceneColorTrack
       display_name=Color and Opacity
       sections [{
         type=MovieSceneColorSection
         end_frame=6
         start_frame=0
         channels [
          Color.R {type=MovieSceneScriptingFloatChannel, keys [[0,0.4125],[6,1]]}
          Color.G {type=MovieSceneScriptingFloatChannel, keys [[0,0.5089],[6,1]]}
          Color.B {type=MovieSceneScriptingFloatChannel, keys [[0,0.9647],[6,1]]}
          Color.A {type=MovieSceneScriptingFloatChannel, const [1,2]}
         ]
        }]
      }]
    }
    Text_Rank {
     id=#3
     possessed_class=CommonTextBlock
     tracks [{
       type=MovieSceneColorTrack
       display_name=Color and Opacity
       sections [{
         type=MovieSceneColorSection
         end_frame=6
         start_frame=0
         channels [
          Color.R {type=MovieSceneScriptingFloatChannel, keys [[0,0.4125],[6,1]]}
          Color.G {type=MovieSceneScriptingFloatChannel, keys [[0,0.5089],[6,1]]}
          Color.B {type=MovieSceneScriptingFloatChannel, keys [[0,0.9647],[6,1]]}
          Color.A {type=MovieSceneScriptingFloatChannel, const [1,1]}
         ]
        }]
      }]
    }
    GlobalIcon {
     id=#7
     possessed_class=Image
     tracks [{
       type=MovieSceneColorTrack
       display_name=Color and Opacity
       sections [{
         type=MovieSceneColorSection
         end_frame=6
         start_frame=0
         channels [
          Color.R {type=MovieSceneScriptingFloatChannel, keys [[0,0.4125],[6,1]]}
          Color.G {type=MovieSceneScriptingFloatChannel, keys [[0,0.5089],[6,1]]}
          Color.B {type=MovieSceneScriptingFloatChannel, keys [[0,0.9647],[6,1]]}
          Color.A {type=MovieSceneScriptingFloatChannel, const [1,2]}
         ]
        }]
      }]
    }
   ]}
 }
]
graphs [
 EventGraph {nodes [
   EnableHighlightForLocalPlayer() {is=title K2Node_CustomEvent}
   > PlayAnimation(InAnimation=Get PlayerEntry() -> (PlayerEntry:Widget Animation Object Reference))
   n1 = SetEntryData!() -> (OutputDelegate, then, LeaderboardEntryData:Pilgrim Leaderboard Entry Structure) {is=title K2Node_CustomEvent}
   > n2 = GetEntryName!(execute, self, LeaderboardEntryData:Pilgrim Leaderboard Entry Structure=@n1.LeaderboardEntryData) -> (then, EntryName:Text)
   > "SetText (Text)"(self=Get Text_Team() -> (Text_Team:Common Text Object Reference), InText=@n2)
   n3 = Break Pilgrim Leaderboard Entry!(PilgrimLeaderboardEntry:"Pilgrim Leaderboard Entry Structure (by ref)"=@n1.LeaderboardEntryData) -> (
    bHasLocalPlayer:bool=~
    PlayerAccountIdToDisplayNameMap:Map of Strings to Strings
    Rank:int
    Percentile:float=~
    ScoreData:Pilgrim Leaderboard Score Data Structure
    DateTime:Date Time Structure
    FriendRank:int
   ) {is=title K2Node_BreakStruct}
   n4 = Break Pilgrim Leaderboard Score Data!(PilgrimLeaderboardScoreData:"Pilgrim Leaderboard Score Data Structure (by ref)"=@n3.ScoreData) -> (
    Score:int
    StarsEarned:int
    Accuracy:float=~
    bFullCombo:bool=~
    Difficulty:EPilgrimSongDifficulty Enum
    InstrumentsPlayed:Array of EPilgrimTrackType Enums
    SeasonNumber:int
   ) {is=title K2Node_BreakStruct}
   > "SetText (Text)"(self=Get Text_Score() -> (Text_Score:Common Text Object Reference), InText="To Text (Integer)"(Value=@n4))
   > SetAccuracy!(execute, self, Accuracy:double=@n4.Accuracy, Full Combo:bool=@n4.bFullCombo) -> (then)
   > SetDifficulty!(execute, self, Difficulty:EPilgrimSongDifficulty Enum=@n4.Difficulty) -> (then)
   > SetRankIcon!(execute, self, FriendRank:int=@n3.FriendRank) -> (then)
   n5 = Event OnListItemObjectSet()
   > n6 = Cast To BP_PilgrimQuickplay_LeaderboardEntry_ListObject!(execute, Object=@n5.ListItemObject) -> (then, CastFailed, AsBP Pilgrim Quickplay Leaderboard Entry List Object:BP Pilgrim Quickplay Leaderboard Entry List Object Object Reference) {is=title K2Node_DynamicCast}
   > Set Entry Data From List Object!(execute, self, List Object:BP Pilgrim Quickplay Leaderboard Entry List Object Object Reference=@n6) -> (then) {is=title K2Node_CallFunction}
  ]}
 GetEntryName {nodes [
   n1 = GetEntryName!() -> (
    then
    LeaderboardEntryData_bHasLocalPlayer:bool=~
    LeaderboardEntryData_PlayerAccountIdToDisplayNameMap:Map of Strings to Strings
    LeaderboardEntryData_Rank:int
    LeaderboardEntryData_Percentile:float=~
    LeaderboardEntryData_ScoreData:Pilgrim Leaderboard Score Data Structure
    LeaderboardEntryData_DateTime:Date Time Structure
    LeaderboardEntryData_FriendRank:int
   ) {is=title K2Node_FunctionEntry}
   > n2 = Values!(execute, TargetMap:Map of Strings to Strings=@n1.LeaderboardEntryData_PlayerAccountIdToDisplayNameMap) -> (then, Values:Array of Strings)
   > n3 = Branch(Condition="Equal (Integer)"(A=Length(TargetArray:Array of Strings=@n2), B=1))
   n4 = "Get (a copy)"(Array=@n2)
   > Return Node!(
    execute
    EntryName:Text=Select(
     Option 0:Text="To Text (String)"(InString=@n4)
     Option 1:Text=Get EmptyNameFallbackText() -> (EmptyNameFallbackText:Text)
     Index=IsEmpty(InString=@n4)
    ) -> (ReturnValue:Text)
   )
   n5 = "Get (a copy)"(Array=@n2)
   n6 = Format Text!(
    Format="LOCTEXT(\"{Name}\\'s\")"
    Name:Text=Select(Option 0="To Text (String)"(InString=@n5), Option 1=Get EmptyNameFallbackText(), Index=IsEmpty(InString=@n5))
   ) -> (Result)
   n7 = Get IsPlayerEntry() -> (IsPlayerEntry:bool)
   Return Node(
    execute=@n3.else
    EntryName=Format Text(Format="LOCTEXT(\"{Name} band\")", Name=Select(Option 0=@n6, Option 1="LOCTEXT(\"Your\")", Index=@n7))
   )
   n8 = GetOwningPlayer()
   "To Text (String)"(InString=GetDisplayName(Object=GetFortPlayerControllerFromActor!(Actor:Actor Object Reference=@n8) -> (ReturnValue:Fort Player Controller Object Reference)))
  ]}
 Set Entry Data From List Object {nodes [
   n1 = Set Entry Data From List Object!() -> (then, List Object:BP Pilgrim Quickplay Leaderboard Entry List Object Object Reference) {is=title K2Node_FunctionEntry}
   > n2 = Sequence!(execute) -> (then_0, then_1, then_2:Exec, then_3:Exec)
   n3 = Get LeaderboardEntryData!(self:BP Pilgrim Quickplay Leaderboard Entry List Object Object Reference=@n1) -> (LeaderboardEntryData:Pilgrim Leaderboard Entry Structure)
   n4 = Break Pilgrim Leaderboard Entry(PilgrimLeaderboardEntry=@n3) -> (bHasLocalPlayer=~, Percentile=~)
   n5 = SelectInt(A=@n4.Rank, B=@n4.FriendRank, bPickA="integer <= integer"(A=@n4.FriendRank))
   > "SetText (Text)"(
    self=Get Text_Rank() -> (Text_Rank:Common Text Object Reference)
    InText=Format Text!(Format="LOCTEXT(\"{Rank}\")", Rank:int=@n5) -> (Result)
   )
   n6 = "% (Integer)"(A=int + int(A=@n5), B=2) -> (ReturnValue=0)
   > SetBrushColor(
    self=Get CommonBorder_Zebra() -> (CommonBorder_Zebra:Common Border Object Reference)
    InBrushColor=Select(
     Option 0:Linear Color Structure=Make Linear Color!(R:float=0.01, G:float=0, B:float=0.15, A:float=0.2) -> (LinearColor:Linear Color Structure) {is=title K2Node_MakeStruct}
     Option 1:Linear Color Structure=(R=0,G=0,B=0,A=0)
     Index="Equal (Integer)"(A=@n6, B=0) {is=title K2Node_CallFunction}
    ) -> (ReturnValue:Linear Color Structure)
   )
   Branch(execute=@n2.then_1, Condition=@n4)
   > EnableHighlightForLocalPlayer() {is=title K2Node_CallFunction}
   SetEntryData!(execute=@n2.then_2, self, LeaderboardEntryData:Pilgrim Leaderboard Entry Structure=@n3) -> (then) {is=title K2Node_CallFunction}
   Reroute Node()
   > SetVisibility(InVisibility=Select(
     Option 0:ESlateVisibility Enum=Collapsed
     Option 1:ESlateVisibility Enum=SelfHitTestInvisible
     Index="Equal (Integer)"(B=1) {is=title K2Node_PromotableOperator}
    ) -> (ReturnValue:ESlateVisibility Enum))
   Reroute Node(InputPin:Wildcard) -> (OutputPin:Wildcard)
  ]}
 SetAccuracy {nodes [
   n1 = SetAccuracy!() -> (then, Accuracy:double=~, Full Combo:bool=~) {is=title K2Node_FunctionEntry}
   > "SetText (Text)"(
    self=Get Text_Accuracy() -> (Text_Accuracy:Common Text Object Reference)
    InText=Format Text!(Format="LOCTEXT(\"{Accuracy}%\")", Accuracy:double=@n1) -> (Result)
   )
   > SetActiveWidgetIndex(
    self:Common Visibility Switcher Object Reference=Get FullCombo_Switcher() -> (FullCombo_Switcher:Common Visibility Switcher Object Reference)
    Index="To Integer (Boolean)"!(InBool:bool=@n1.Full Combo) -> (ReturnValue:int=0)
   )
  ]}
 SetDifficulty {nodes [
   n1 = SetDifficulty!() -> (then, Difficulty:EPilgrimSongDifficulty Enum=DifficultyEasy) {is=title K2Node_FunctionEntry}
   > "SetText (Text)"(
    self=Get CommonTextBlock-Difficulty!(self:WBP Pilgrim Results Difficulty Object Reference=Get WBP_Pilgrim_Results_Difficulty() -> (WBP_Pilgrim_Results_Difficulty:WBP Pilgrim Results Difficulty Object Reference)) -> (CommonTextBlock-Difficulty:Common Text Object Reference)
    InText=Select!(
     DifficultyEasy:"Text (by ref)"=MakeLiteralText(Value="LOCTEXT(\"E\")")
     DifficultyMedium:"Text (by ref)"=MakeLiteralText(Value="LOCTEXT(\"M\")")
     DifficultyHard:"Text (by ref)"=MakeLiteralText(Value="LOCTEXT(\"H\")")
     DifficultyExpert:"Text (by ref)"=MakeLiteralText(Value="LOCTEXT(\"X\")")
     None:"Text (by ref)"=MakeLiteralText(Value="LOCTEXT(\"N\")")
     Index:EPilgrimSongDifficulty Enum=@n1
    ) -> (ReturnValue:"Text (by ref)")
   )
  ]}
 SetRankIcon {nodes [
   n1 = SetRankIcon!() -> (then, FriendRank:int) {is=title K2Node_FunctionEntry}
   > n2 = Branch(Condition="integer <= integer"(A=@n1))
   n3 = Get RankIconSwitcher() -> (RankIconSwitcher:Common Visibility Switcher Object Reference)
   > SetActiveWidgetIndex(self=@n3, Index=0)
   SetActiveWidgetIndex(execute=@n2.else, self=@n3, Index=1)
  ]}
]
variables [
 EmptyNameFallbackText {type=string, default=No Name Found}
 IsPlayerEntry {type=boolean}
]
widgets {
 info {inheritedWidgetCount=0, parentClass=&CommonUserWidget, rootWidgetClass=&Overlay, widgetCount=18}
 widgets [
  Overlay_99 {widgetClassPath=&Overlay, widget=Overlay_99}
  CurrentUserBG {
   widgetClassPath=&Border
   bIsVariable=true
   parent=Overlay_99
   widget=CurrentUserBG
   properties {
    renderOpacity=0.75
    visibility=Hidden
    background {
     bIsDynamicallyLoaded=false
     drawAs=RoundedBox
     imageType=FullColor
     mirroring=NoMirror
     resourceName=None
     resourceObject=/PilgrimQuickplay/Widgets/Leaderboards/Materials/MI_Leaderboard_CurrentUser_Purple
     tiling=NoTile
     imageSize=<32 32>
     margin=<ltrb 0 0 0 0>
     outlineSettings {
      bUseBrushTransparency=false
      roundingType=FixedRadius
      width=0
      color {colorUseRule=UseColor_Specified, specifiedColor=<rgba 0 0 0 0>}
      cornerRadii=<16 16 16 16>
     }
     tintColor {colorUseRule=UseColor_Specified, specifiedColor=<rgba 1 1 1 1>}
     uVRegion {bIsValid=false, max=<0 0>, min=<0 0>}
    }
    padding=<ltrb 0 0 0 0>
   }
   slot {class=OverlaySlot, properties {horizontalAlignment=HAlign_Fill, verticalAlignment=VAlign_Fill}}
  }
  SizeBox_33 {
   widgetClassPath=&SizeBox
   parent=Overlay_99
   widget=SizeBox_33
   properties {bOverride_MinDesiredHeight=true, minDesiredHeight=48}
   slot {class=OverlaySlot, properties {horizontalAlignment=HAlign_Fill, verticalAlignment=VAlign_Fill}}
  }
  CommonBorder_Zebra {
   widgetClassPath=&CommonBorder
   bIsVariable=true
   parent=SizeBox_33
   widget=CommonBorder_Zebra
   properties {
    background {
     bIsDynamicallyLoaded=false
     drawAs=RoundedBox
     imageType=FullColor
     mirroring=NoMirror
     resourceName=None
     resourceObject=&MI_Leaderboard_CurrentUser_Purple
     tiling=NoTile
     imageSize=<32 32>
     margin=<ltrb 0 0 0 0>
     outlineSettings {
      bUseBrushTransparency=false
      roundingType=FixedRadius
      width=0
      color {colorUseRule=UseColor_Specified, specifiedColor=<rgba 0 0 0 0>}
      cornerRadii=<16 16 16 16>
     }
     tintColor {colorUseRule=UseColor_Specified, specifiedColor=<rgba 1 1 1 1>}
     uVRegion {bIsValid=false, max=<0 0>, min=<0 0>}
    }
    padding=<ltrb 15 0 15 0>
   }
   slot {class=SizeBoxSlot}
  }
  HorizontalBox_86 {
   widgetClassPath=&HorizontalBox
   parent=CommonBorder_Zebra
   widget=HorizontalBox_86
   slot {class=BorderSlot, properties {padding=<ltrb 15 0 15 0>}}
  }
  RankIconSwitcher {
   widgetClassPath=&CommonVisibilitySwitcher
   bIsVariable=true
   parent=HorizontalBox_86
   widget=RankIconSwitcher
   slot {class=HorizontalBoxSlot, properties {verticalAlignment=VAlign_Center, padding=<ltrb 0 5 10 5>}}
  }
  GlobalIcon {
   widgetClassPath=&Image
   bIsVariable=true
   parent=RankIconSwitcher
   widget=GlobalIcon
   properties {
    visibility=SelfHitTestInvisible
    brush {
     bIsDynamicallyLoaded=false
     drawAs=Image
     imageType=FullColor
     mirroring=NoMirror
     resourceName=None
     resourceObject=/Game/UI/Friends_UI/Social/Globe_PlatformIcon_32x
     tiling=NoTile
     imageSize=<24 24>
     margin=<ltrb 0 0 0 0>
     outlineSettings {
      bUseBrushTransparency=false
      roundingType=HalfHeightRadius
      width=0
      color {colorUseRule=UseColor_Specified, specifiedColor=<rgba 0 0 0 0>}
      cornerRadii=<0 0 0 0>
     }
     tintColor {colorUseRule=UseColor_Specified, specifiedColor=<rgba 1 1 1 1>}
     uVRegion {bIsValid=false, max=<0 0>, min=<0 0>}
    }
    colorAndOpacity=<rgba 0.3916 0.4852 0.9387 1>
   }
   slot {class=CommonVisibilitySwitcherSlot}
  }
  FriendIcon {
   widgetClassPath=&Image
   bIsVariable=true
   parent=RankIconSwitcher
   widget=FriendIcon
   properties {
    visibility=SelfHitTestInvisible
    brush {
     bIsDynamicallyLoaded=false
     drawAs=Image
     imageType=FullColor
     mirroring=NoMirror
     resourceName=None
     resourceObject=/Game/UI/Foundation/Shell/Textures/T-Icon-Friends-L
     tiling=NoTile
     imageSize=<24 24>
     margin=<ltrb 0 0 0 0>
     outlineSettings {
      bUseBrushTransparency=false
      roundingType=HalfHeightRadius
      width=0
      color {colorUseRule=UseColor_Specified, specifiedColor=<rgba 0 0 0 0>}
      cornerRadii=<0 0 0 0>
     }
     tintColor {colorUseRule=UseColor_Specified, specifiedColor=<rgba 1 1 1 1>}
     uVRegion {bIsValid=false, max=<0 0>, min=<0 0>}
    }
    colorAndOpacity=<rgba 0.4125 0.5089 0.9647 1>
   }
   slot {class=CommonVisibilitySwitcherSlot}
  }
  Text_Rank {
   widgetClassPath=&CommonTextBlock
   bIsVariable=true
   parent=HorizontalBox_86
   widget=Text_Rank
   properties {
    text=2
    colorAndOpacity {colorUseRule=UseColor_Specified, specifiedColor=<rgba 0.4125 0.5089 0.9647 1>}
    font {
     bForceMonospaced=false
     bMaterialIsStencil=false
     fontMaterial=null
     fontObject=&HeadingNow
     letterSpacing=0
     monospacedWidth=1
     size=15
     skewAmount=0
     typefaceFontName=74 Regular
     outlineSettings {
      bApplyOutlineToDropShadows=false
      bMiteredCorners=false
      bSeparateFillAlpha=false
      outlineMaterial=null
      outlineSize=0
      outlineColor=<rgba 0 0 0 1>
     }
    }
    shadowOffset=<0 0>
   }
   slot {class=HorizontalBoxSlot, properties {horizontalAlignment=HAlign_Left, verticalAlignment=VAlign_Center}}
  }
  Text_Team {
   widgetClassPath=&CommonTextBlock
   bIsVariable=true
   parent=HorizontalBox_86
   widget=Text_Team
   properties {
    scrollStyle=&TextScrollStyle-Base_C
    text=Username
    colorAndOpacity {colorUseRule=UseColor_Specified, specifiedColor=<rgba 0.3916 0.4852 0.9387 1>}
    font {
     bForceMonospaced=false
     bMaterialIsStencil=false
     fontMaterial=null
     fontObject=&HeadingNow
     letterSpacing=0
     monospacedWidth=1
     size=15
     skewAmount=0
     typefaceFontName=74 Regular
     outlineSettings {
      bApplyOutlineToDropShadows=false
      bMiteredCorners=false
      bSeparateFillAlpha=false
      outlineMaterial=null
      outlineSize=0
      outlineColor=<rgba 0 0 0 1>
     }
    }
    shadowOffset=<0 0>
   }
   slot {
    class=HorizontalBoxSlot
    properties {horizontalAlignment=HAlign_Left, verticalAlignment=VAlign_Center, padding=<ltrb 15 0 0 0>, size {sizeRule=Fill, value=1}}
   }
  }
  Spacer_213 {
   widgetClassPath=&Spacer
   parent=HorizontalBox_86
   widget=Spacer_213
   slot {class=HorizontalBoxSlot, properties {size {sizeRule=Fill, value=1}}}
  }
  HorizontalBox_233 {
   widgetClassPath=&HorizontalBox
   parent=HorizontalBox_86
   widget=HorizontalBox_233
   slot {class=HorizontalBoxSlot, properties {horizontalAlignment=HAlign_Right, verticalAlignment=VAlign_Center}}
  }
  FullCombo_Switcher {
   widgetClassPath=&CommonVisibilitySwitcher
   bIsVariable=true
   parent=HorizontalBox_233
   widget=FullCombo_Switcher
   slot {class=HorizontalBoxSlot, properties {verticalAlignment=VAlign_Center, padding=<ltrb 0 0 12 0>}}
  }
  Text_Accuracy {
   widgetClassPath=&CommonTextBlock
   bIsVariable=true
   parent=FullCombo_Switcher
   widget=Text_Accuracy
   properties {
    text=100%
    colorAndOpacity {colorUseRule=UseColor_Specified, specifiedColor=<rgba 0.4125 0.5089 0.9647 1>}
    font {
     bForceMonospaced=false
     bMaterialIsStencil=false
     fontMaterial=null
     fontObject=&HeadingNow
     letterSpacing=0
     monospacedWidth=1
     size=15
     skewAmount=0
     typefaceFontName=74 Regular
     outlineSettings {
      bApplyOutlineToDropShadows=false
      bMiteredCorners=false
      bSeparateFillAlpha=false
      outlineMaterial=null
      outlineSize=0
      outlineColor=<rgba 0 0 0 1>
     }
    }
    shadowOffset=<0 0>
   }
   slot {class=CommonVisibilitySwitcherSlot, properties {verticalAlignment=VAlign_Center, padding=<ltrb 0 0 15 0>}}
  }
  WBP_PilgrimQuickplay_FullComboIndicator_1 {
   widgetClassPath=/PilgrimQuickplay/Widgets/WBP_PilgrimQuickplay_FullComboIndicator
   bIsVariable=true
   parent=FullCombo_Switcher
   widget=WBP_PilgrimQuickplay_FullComboIndicator_1
   slot {class=CommonVisibilitySwitcherSlot}
  }
  WBP_Pilgrim_Results_Difficulty {
   widgetClassPath=/PilgrimQuickplay/Widgets/Results/WBP_Pilgrim_Results_Difficulty
   bIsVariable=true
   parent=HorizontalBox_233
   widget=WBP_Pilgrim_Results_Difficulty
   slot {class=HorizontalBoxSlot, properties {horizontalAlignment=HAlign_Center, verticalAlignment=VAlign_Center, padding=<ltrb 6 0 3 0>}}
  }
  SizeBox_0 {
   widgetClassPath=&SizeBox
   parent=HorizontalBox_233
   widget=SizeBox_0
   properties {bOverride_MinDesiredWidth=true, minDesiredWidth=130, widthOverride=130}
   slot {class=HorizontalBoxSlot, properties {verticalAlignment=VAlign_Center, padding=<ltrb 6 0 0 0>}}
  }
  Text_Score {
   widgetClassPath=&CommonTextBlock
   bIsVariable=true
   parent=SizeBox_0
   widget=Text_Score
   properties {
    justification=Center
    text="9,999,999"
    colorAndOpacity {colorUseRule=UseColor_Specified, specifiedColor=<rgba 0.4125 0.5089 0.9647 1>}
    font {
     bForceMonospaced=true
     bMaterialIsStencil=false
     fontMaterial=null
     fontObject=&HeadingNow
     letterSpacing=0
     monospacedWidth=1
     size=15
     skewAmount=0
     typefaceFontName=74 Regular
     outlineSettings {
      bApplyOutlineToDropShadows=false
      bMiteredCorners=false
      bSeparateFillAlpha=false
      outlineMaterial=null
      outlineSize=0
      outlineColor=<rgba 0 0 0 1>
     }
    }
    shadowOffset=<0 0>
   }
   slot {class=SizeBoxSlot, properties {horizontalAlignment=HAlign_Right, verticalAlignment=VAlign_Center}}
  }
 ]
}
