WidgetBlueprint /PilgrimQuickplay/Widgets/Leaderboards/WBP_PilgrimQuickplay_Leaderboard_Mini
animations [
 BeatScoreText_Intro {
  binding_count=1
  length_seconds=1.5
  object=&self:BeatScoreText_Intro
  sequence {bindings [ScoreBeatenTextBorder {
     id=#1
     possessed_class=CommonBorder
     tracks [
      {
       type=MovieSceneColorTrack
       display_name=Brush Color
       sections [{
         type=MovieSceneColorSection
         end_frame=30
         start_frame=0
         channels [
          Color.R {type=MovieSceneScriptingFloatChannel, const [1,7]}
          Color.G {type=MovieSceneScriptingFloatChannel, keys [[0,0.9866],[5,1],[10,0.9866],[15,1],[20,0.9866],[25,1],[30,0.9866]]}
          Color.B {type=MovieSceneScriptingFloatChannel, keys [[0,0],[5,1],[10,0],[15,1],[20,0],[25,1],[30,0]]}
          Color.A {type=MovieSceneScriptingFloatChannel, const [1,7]}
         ]
        }]
      }
      {
       type=MovieSceneDoubleVectorTrack
       display_name=Desired Size Scale
       sections [{
         type=MovieSceneDoubleVectorSection
         end_frame=30
         start_frame=-1
         channels [
          Vector.X {type=MovieSceneScriptingDoubleChannel, keys [[0,1],[5,2],[25,2],[30,1]]}
          Vector.Y {type=MovieSceneScriptingDoubleChannel, keys [[0,0],[5,2],[25,2],[30,1]]}
         ]
        }]
      }
     ]
    }]}
 }
 BeatScoreText_Idle {
  binding_count=1
  length_seconds=3
  object=&self:BeatScoreText_Idle
  sequence {bindings [ScoreBeatenTextBorder {
     id=#1
     possessed_class=CommonBorder
     tracks [{
       type=MovieSceneColorTrack
       display_name=Brush Color
       sections [{
         type=MovieSceneColorSection
         end_frame=60
         start_frame=0
         channels [
          Color.R {type=MovieSceneScriptingFloatChannel, const [1,4]}
          Color.G {type=MovieSceneScriptingFloatChannel, keys [[0,0.9866],[50,0.9866],[55,1],[60,0.9866]]}
          Color.B {type=MovieSceneScriptingFloatChannel, keys [[0,0],[50,0],[55,1],[60,0]]}
          Color.A {type=MovieSceneScriptingFloatChannel, const [1,4]}
         ]
        }]
      }]
    }]}
 }
]
graphs [
 CalculateIndicesAroundPlayerToDisplay {nodes [
   n1 = CalculateIndicesAroundPlayerToDisplay!() -> (then, TotalEntries:int) {is=title K2Node_FunctionEntry}
   > n2 = Sequence()
   n3 = Get EntriesPlusMinusAroundPlayer() -> (EntriesPlusMinusAroundPlayer:int=0)
   n4 = Get PlayerEntryIndex() -> (PlayerEntryIndex:int=0)
   > Set StartingDisplayIndex(StartingDisplayIndex:int=int - int(A=@n4, B=@n3)) -> (Output_Get:int=0)
   > Set LastDisplayIndex(LastDisplayIndex:int=int + int(A=@n4, B=@n3)) -> (Output_Get:int=0)
   n5 = Branch(execute=[@n6."  ", @n2.then_1], Condition="integer < integer"(A=Get StartingDisplayIndex() -> (StartingDisplayIndex:int=0)))
   > Increment Int(Value=Get StartingDisplayIndex() -> (StartingDisplayIndex=0))
   > n6 = Increment Int(Value=Get LastDisplayIndex() -> (LastDisplayIndex:int=0))
   n7 = int - int(A=@n1)
   n8 = Branch(execute=[@n5.else, @n9."  "], Condition="integer > integer"(A=Get LastDisplayIndex() -> (LastDisplayIndex=0), B=@n7))
   > Decrement Int(Value=Get LastDisplayIndex() -> (LastDisplayIndex=0))
   > n9 = Decrement Int(Value=Get StartingDisplayIndex() -> (StartingDisplayIndex=0))
   Set StartingDisplayIndex(
    execute=@n8.else
    StartingDisplayIndex="Clamp (Integer)"(Value=Get StartingDisplayIndex() -> (StartingDisplayIndex=0), Min=0, Max=@n7) -> (ReturnValue=0)
   ) -> (Output_Get=0)
   > Set LastDisplayIndex(LastDisplayIndex="Clamp (Integer)"(Value=Get LastDisplayIndex() -> (LastDisplayIndex=0), Min=0, Max=@n7) -> (ReturnValue=0)) -> (Output_Get=0)
  ]}
 EventGraph {nodes [
   n1 = GetScoreBeatenText!(execute, self) -> (then, ScoreBeatenText:Text)
   > "SetText (Text)"(self=Get ScoreBeatenText() -> (ScoreBeatenText:Common Text Object Reference), InText=@n1)
   > Delay(Duration=0.5)
   > n2 = "Play Animation with Finished event (legacy)"!(
    execute:Exec
    Widget:User Widget Object Reference=Self-Reference()
    InAnimation:Widget Animation Object Reference=Get BeatScoreText_Intro() -> (BeatScoreText_Intro:Widget Animation Object Reference)
    StartAtTime:float=0
    NumLoopsToPlay:int=1
    PlayMode:EUMGSequencePlayMode Enum
    PlaybackSpeed:float=1
   ) -> (then:Exec, Result:UMGSequence Player Object Reference, Finished:Exec) {is=title K2Node_PlayAnimation}
   PlayAnimation(execute=@n2.Finished, InAnimation=Get BeatScoreText_Idle() -> (BeatScoreText_Idle:Widget Animation Object Reference), NumLoopsToPlay=0)
   RequestCenteredLeaderboardData() {is=title K2Node_CustomEvent}
   n3 = Get CenteredLeaderboardRequest() -> (CenteredLeaderboardRequest:Pilgrim Async Action Request Leaderboard Object Reference)
   > n4 = Is Valid(InputObject=@n3)
   > n5 = Cancel(self:Cancellable Async Action Object Reference=@n3)
   n6 = Set RemainingCenteredRetries(execute=[@n5.then, @n4.Is Not Valid], RemainingCenteredRetries:int=Get MaxRetries() -> (MaxRetries:int=0)) -> (Output_Get:int=0)
   Branch(execute=[@n6.then, @n12."  "], Condition="Not Equal (Name)"(A=Get SongShortName() -> (SongShortName:Name=None)) -> (ReturnValue=~))
   > n7 = PilgrimRequestLeaderboard!(
    execute:Exec
    SongShortName:Name=Get SongShortName() -> (SongShortName=None)
    SpotlightGroup:String
    LeaderboardType:EPilgrimLeaderboardType Enum=Get LeaderboardType() -> (LeaderboardType:EPilgrimLeaderboardType Enum=SoloGuitar)
    RequestingPlayer:Fort Player Controller Object Reference=Get RequestingPlayerController() -> (RequestingPlayerController:Fort Player Controller Object Reference)
    bFriendsOnly:bool=Get ShouldRequestFriendsLeaderboard() -> (ShouldRequestFriendsLeaderboard:bool)
    bCenteredOnRequestingPlayer:bool=true
    Page:int=0
    bIsSpotlightCompetition:bool=false
    bIsAllTimeHighScore:bool=false
    bUseCache:bool=true
    FreshnessTime:int=-1
    RequestPriority:EPilgrimLeaderboardRequestPriority Enum=Normal
   ) -> (
    then:Exec
    AsyncTaskProxy:Pilgrim Async Action Request Leaderboard Object Reference
    OnSuccess:Exec
    OnFailure:Exec
    ResultPage:Pilgrim Leaderboard Page Structure
    RequestParams:Pilgrim Leaderboard Request Params Structure
   ) {is=title K2Node_AsyncAction}
   > Set CenteredLeaderboardRequest(CenteredLeaderboardRequest:Pilgrim Async Action Request Leaderboard Object Reference=@n7) -> (Output_Get:Pilgrim Async Action Request Leaderboard Object Reference)
   n8 = Get RemainingCenteredRetries() -> (RemainingCenteredRetries:int=0)
   PrintString!(
    execute=@n7.OnFailure
    InString=Append!(A="Centered leaderboard request failed, ", B="To String (Integer)"(InInt=@n8), C:String=" retries remaining") -> (ReturnValue)
    bPrintToScreen
    bPrintToLog
    TextColor
    Duration
    Key
   ) -> (then)
   n9 = Get RemainingCenteredRetries() -> (RemainingCenteredRetries=0)
   > n10 = Branch(Condition="integer < integer"(A=@n9, B=1))
   > n11 = DisplayTopEntries()
   Set CenteredLeaderboardRequest(execute=[@n11.then, @n14.then, @n13.else])
   n12 = Decrement Int(" "=@n10.else, Value=@n9)
   Set CenteredLeaderboardPage(execute=@n7.OnSuccess, CenteredLeaderboardPage:Pilgrim Leaderboard Page Structure=@n7.ResultPage) -> (Output_Get:Pilgrim Leaderboard Page Structure)
   > n13 = Branch(Condition=Get RequestToDisplayCenteredDataHasBeenMade() -> (RequestToDisplayCenteredDataHasBeenMade:bool))
   > n14 = DisplayCenteredLeaderboardData()
   DisplayCenteredLeaderboardData() {is=title K2Node_CustomEvent}
   n15 = Break Pilgrim Leaderboard Page!(PilgrimLeaderboardPage:"Pilgrim Leaderboard Page Structure (by ref)"=Get CenteredLeaderboardPage() -> (CenteredLeaderboardPage:Pilgrim Leaderboard Page Structure)) -> (
    SongShortName:Name
    LeaderboardType:EPilgrimLeaderboardType Enum
    Page:int
    TotalPages:int
    Entries:Array of Pilgrim Leaderboard Entry Structures
    bIsSpotlightCompetition:bool=~
    SpotlightGroup:String
    bIsAllTimeHighScore:bool=~
   ) {is=title K2Node_BreakStruct}
   > n16 = Branch(Condition="Not Equal (Name)"(A=@n15) -> (ReturnValue=~))
   > Set PlayerPageNumber(PlayerPageNumber:int=@n15.Page) -> (Output_Get:int=0)
   > Set Total Pages(Total Pages:int=@n15.TotalPages) -> (Output_Get:int=0)
   > n17 = For Each Loop with Break!(Exec:Exec, Array:Array of Pilgrim Leaderboard Entry Structures=@n15.Entries, Break:Exec=@n23.then) -> (LoopBody:Exec, Array Element:Pilgrim Leaderboard Entry Structure, Array Index:int, Completed:Exec)
   n18 = Branch(execute=@n17.Completed, Condition="integer < integer"(A=Get PlayerEntryIndex() -> (PlayerEntryIndex=0)))
   > DisplayTopEntries()
   SetActiveWidget(
    execute=@n18.else
    self=Get ContentSwitcher() -> (ContentSwitcher:Common Visibility Switcher Object Reference)
    Widget=Get LeaderboardContainer() -> (LeaderboardContainer:Vertical Box Object Reference)
   )
   > n19 = Construct Object from Class!(
    execute
    Class=&/BP_PilgrimQuickplay_LeaderboardEntry_Mini_ListObject.BP_PilgrimQuickplay_LeaderboardEntry_Mini_ListObject_C
    self
    EntryDataStruct:Pilgrim Leaderboard Entry Structure="Get (a copy)"(Array:Array of Pilgrim Leaderboard Entry Structures=Get FirstPageLeaderboardEntries() -> (FirstPageLeaderboardEntries:Array of Pilgrim Leaderboard Entry Structures)) -> (Output:Pilgrim Leaderboard Entry Structure)
    ListIndex:int=0
    IsPlayerEntry:bool=False
    ShouldShowDivider:bool=true
   ) -> (then, ReturnValue:BP Pilgrim Quickplay Leaderboard Entry Mini List Object Object Reference)
   > AddItem!(
    execute
    self:List View Object Reference=Get LeaderboardEntries() -> (LeaderboardEntries:Common List View Object Reference)
    Item:Object Reference=@n19
   ) -> (then)
   > CalculateIndicesAroundPlayerToDisplay!(execute, self, TotalEntries:int=Length(TargetArray:Array of Pilgrim Leaderboard Entry Structures=@n15.Entries)) -> (then) {is=title K2Node_CallFunction}
   > n20 = For Loop(FirstIndex=Get StartingDisplayIndex() -> (StartingDisplayIndex=0), LastIndex=Get LastDisplayIndex() -> (LastDisplayIndex=0))
   UpdateScoreBeatenText(execute=@n20.Completed)
   Branch(
    execute=@n20.LoopBody
    Condition=Is Valid Index!(TargetArray:Array of Pilgrim Leaderboard Entry Structures=@n15.Entries, IndexToTest:int=@n20) -> (ReturnValue:bool)
   )
   > n21 = Construct Object from Class(
    Class=&BP_PilgrimQuickplay_LeaderboardEntry_Mini_ListObject_C
    EntryDataStruct="Get (a copy)"(Array=@n15.Entries, Dimension 1=@n20)
    ListIndex=@n20
    IsPlayerEntry="Equal (Integer)"(A=@n20, B=Get PlayerEntryIndex() -> (PlayerEntryIndex=0))
    ShouldShowDivider=False
   )
   > AddItem(self=Get LeaderboardEntries(), Item=@n21)
   Branch(
    execute=@n17.LoopBody
    Condition=Break Pilgrim Leaderboard Entry!(PilgrimLeaderboardEntry:"Pilgrim Leaderboard Entry Structure (by ref)"=@n17) -> (
     bHasLocalPlayer:bool=~
     PlayerAccountIdToDisplayNameMap:Map of Strings to Strings
     Rank:int
     Percentile:float=~
     ScoreData:Pilgrim Leaderboard Score Data Structure
     DateTime:Date Time Structure
     FriendRank:int
    ) {is=title K2Node_BreakStruct}
   )
   > n22 = IsCurrentBandEntry!(execute, self, LeaderboardEntry:Pilgrim Leaderboard Entry Structure=@n17) -> (then, AllBandMembersMatch:bool)
   > Branch(Condition=@n22)
   > n23 = Set PlayerEntryIndex(PlayerEntryIndex:int=@n17.Array Index) -> (Output_Get:int=0)
   Set RequestToDisplayCenteredDataHasBeenMade(execute=@n16.else, RequestToDisplayCenteredDataHasBeenMade:bool=true) -> (Output_Get:bool)
   ShowFailedState() {is=title K2Node_CustomEvent}
   > SetActiveWidget(self=Get ContentSwitcher(), Widget=Get Text_NoLeaderboardEntries() -> (Text_NoLeaderboardEntries:Common Text Object Reference))
   DisplayTopEntries() {is=title K2Node_CustomEvent}
   > n24 = Branch(Condition=Is Empty!(TargetArray:Array of Pilgrim Leaderboard Entry Structures=Get FirstPageLeaderboardEntries()) -> (ReturnValue:bool))
   > ShowFailedState() {is=title K2Node_CallFunction}
   n25 = For Loop(execute=@n24.else, LastIndex=int - int(A=Get NumTopEntriesToShow() -> (NumTopEntriesToShow:int=0)))
   SetActiveWidget(execute=@n25.Completed, self=Get ContentSwitcher(), Widget=Get LeaderboardContainer())
   > UpdateScoreBeatenText()
   n26 = Get FirstPageLeaderboardEntries()
   Branch(execute=@n25.LoopBody, Condition=Is Valid Index(TargetArray=@n26, IndexToTest=@n25))
   n27 = "Get (a copy)"(Array=@n26, Dimension 1=@n25)
   > n28 = Construct Object from Class(
    Class=&BP_PilgrimQuickplay_LeaderboardEntry_Mini_ListObject_C
    EntryDataStruct=@n27
    ListIndex=@n25
    IsPlayerEntry="Equal (Integer)"(A=@n25, B=Get PlayerEntryIndex() -> (PlayerEntryIndex=0))
    ShouldShowDivider="Equal (Integer)"(A=Break Pilgrim Leaderboard Entry(PilgrimLeaderboardEntry=@n27) -> (bHasLocalPlayer=~, Percentile=~).Rank, B=1)
   )
   > AddItem(self=Get LeaderboardEntries(), Item=@n28)
   n29 = SetLeaderboardRequestData!() -> (OutputDelegate, then, NumPlayersInBand:int) {is=title K2Node_CustomEvent}
   > n30 = Cast To FortPlayerController(Object=GetOwningPlayer())
   > n31 = Set RequestingPlayerController(RequestingPlayerController:Fort Player Controller Object Reference=@n30) -> (Output_Get:Fort Player Controller Object Reference)
   > n32 = GetPilgrimQuickplayPlayerStateComponent!(
    execute
    PlayerState:Player State Object Reference=Get PlayerState!(self:Controller Object Reference=@n30) -> (PlayerState:Player State Object Reference)
   ) -> (then, ReturnValue:Pilgrim Quickplay Player State Component Object Reference)
   > n33 = Is Valid(InputObject=@n32)
   ShowFailedState(execute=@n33.Is Not Valid)
   Set PlayersInBand(
    execute=@n33.Is Valid
    PlayersInBand:Array of Player State Object References=GetPlayersInBandByController!(
     self:Pilgrim Quickplay Band State Component Object Reference=GetBandState!(self:Pilgrim Quickplay Player State Component Object Reference=@n32) -> (ReturnValue:Pilgrim Quickplay Band State Component Object Reference)
     Controller:Player Controller Object Reference=@n31
    ) -> (ReturnValue:Array of Player State Object References)
   ) -> (Output_Get:Array of Player State Object References)
   n34 = Get Song Short Name For Leaderboard Request!() -> (SongShortName:Name=None) {is=title K2Node_CallFunction}
   > n35 = Branch(Condition="Not Equal (Name)"(A=@n34) -> (ReturnValue=~))
   > Set SongShortName(SongShortName:Name=@n34) -> (Output_Get:Name=None)
   > n36 = Branch(Condition="integer > integer"(A=@n29.NumPlayersInBand, B=1))
   > n37 = Branch(Condition="In Range (Integer)"!(Value:int=@n29.NumPlayersInBand, Min:int=2, Max:int=4, InclusiveMin:bool=true, InclusiveMax:bool=true) -> (ReturnValue:bool))
   n38 = Select!(
    Option 0:EPilgrimLeaderboardType Enum=Invalid
    Option 1:EPilgrimLeaderboardType Enum=Invalid
    Option 2:EPilgrimLeaderboardType Enum=BandDuo
    Option 3:EPilgrimLeaderboardType Enum=BandTrio
    Option 4:EPilgrimLeaderboardType Enum=BandQuad
    Index:int=@n29.NumPlayersInBand
   ) -> (ReturnValue:EPilgrimLeaderboardType Enum)
   > n39 = Branch(Condition="Not Equal (Enum)"!(A:EPilgrimLeaderboardType Enum=@n38, B:EPilgrimLeaderboardType Enum=Invalid) -> (ReturnValue:bool))
   > n40 = Set LeaderboardType(LeaderboardType:EPilgrimLeaderboardType Enum=@n38) -> (Output_Get:EPilgrimLeaderboardType Enum=SoloGuitar)
   Set IsAllLeaderboardDataGathered(execute=[@n40.then, @n43.then], IsAllLeaderboardDataGathered:bool=true) -> (Output_Get:bool)
   ShowFailedState(execute=@n39.else)
   ShowFailedState(execute=@n37.else)
   n41 = Select!(
    TrackGuitar:EPilgrimLeaderboardType Enum=SoloGuitar
    TrackBass:EPilgrimLeaderboardType Enum=SoloBass
    TrackVocals:EPilgrimLeaderboardType Enum=SoloVocals
    TrackDrum:EPilgrimLeaderboardType Enum=SoloDrum
    TrackPlasticGuitar:EPilgrimLeaderboardType Enum=SoloPeripheralGuitar
    TrackPlasticBass:EPilgrimLeaderboardType Enum=SoloPeripheralBass
    TrackPlasticDrum:EPilgrimLeaderboardType Enum=SoloPeripheralDrum
    TrackProVocals:EPilgrimLeaderboardType Enum=SoloPeripheralVocals
    TrackProCymbals:EPilgrimLeaderboardType Enum=SoloPeripheralCymbal
    TrackEvents:EPilgrimLeaderboardType Enum=Invalid
    TrackNone:EPilgrimLeaderboardType Enum=Invalid
    Index:EPilgrimTrackType Enum=GetInstrumentChoice!(self:Pilgrim Core Player State Component Object Reference=@n32) -> (ReturnValue:EPilgrimTrackType Enum=TrackGuitar)
   ) -> (ReturnValue)
   n42 = Branch(execute=@n36.else, Condition="Not Equal (Enum)"(A=@n41, B=Invalid))
   > n43 = Set LeaderboardType(LeaderboardType=@n41) -> (Output_Get=SoloGuitar)
   ShowFailedState(execute=@n42.else)
   ShowFailedState(execute=@n35.else)
   ShowFailedState(execute=@n30.CastFailed)
   RequestTopEntries() {is=title K2Node_CustomEvent}
   > n44 = Set RemainingRetries(RemainingRetries:int=Get MaxRetries() -> (MaxRetries=0)) -> (Output_Get:int=0)
   n45 = PilgrimRequestLeaderboard(
    execute=[@n44.then, @n49."  "]
    SongShortName=Get SongShortName() -> (SongShortName=None)
    LeaderboardType=Get LeaderboardType() -> (LeaderboardType=SoloGuitar)
    RequestingPlayer=Get RequestingPlayerController()
    bFriendsOnly=Get ShouldRequestFriendsLeaderboard()
    bCenteredOnRequestingPlayer=false
    Page=0
    bIsSpotlightCompetition=false
    bIsAllTimeHighScore=false
    bUseCache=true
    FreshnessTime=-1
    RequestPriority=Normal
   )
   > Set LeaderboardRequest(LeaderboardRequest:Pilgrim Async Action Request Leaderboard Object Reference=@n45) -> (Output_Get:Pilgrim Async Action Request Leaderboard Object Reference)
   > RequestCenteredLeaderboardData() {is=title K2Node_CallFunction}
   n46 = Get RemainingRetries() -> (RemainingRetries:int=0)
   PrintString(execute=@n45.OnFailure, InString=Append(A="Leaderboard request failed, ", B="To String (Integer)"(InInt=@n46), C=" retries remaining"))
   n47 = Get RemainingRetries() -> (RemainingRetries=0)
   > n48 = Branch(Condition="integer < integer"(A=@n47, B=1))
   > ShowFailedState()
   n49 = Decrement Int(" "=@n48.else, Value=@n47)
   n50 = Set FirstPageLeaderboardEntries(
    execute=@n45.OnSuccess
    FirstPageLeaderboardEntries:Array of Pilgrim Leaderboard Entry Structures=Break Pilgrim Leaderboard Page(PilgrimLeaderboardPage=@n45.ResultPage) -> (bIsSpotlightCompetition=~, bIsAllTimeHighScore=~).Entries
   ) -> (Output_Get:Array of Pilgrim Leaderboard Entry Structures)
   n51 = Get ShouldRequestFriendsLeaderboard()
   n52 = Length(TargetArray=@n50)
   n53 = Get MinimumEntriesToDisplayFriendsLeaderboard() -> (MinimumEntriesToDisplayFriendsLeaderboard:int=0)
   > n54 = Branch(Condition=OR Boolean(A="integer >= integer"(A=@n52, B=@n53), B=NOT Boolean(A=@n51)))
   n55 = Length(TargetArray=@n50)
   n56 = Get NumTopEntriesToShow() -> (NumTopEntriesToShow=0)
   > n57 = For Loop with Break!(
    Execute:Exec
    FirstIndex:int
    LastIndex:int="Min (Integer)"(A=int - int(A=@n56), B=int - int(A=@n55)) -> (ReturnValue=0)
    Break:Exec=@n62.then
   ) -> (LoopBody:Exec, Index:int, Completed:Exec) {is=title K2Node_MacroInstance}
   n58 = Branch(execute=@n57.Completed, Condition=Get LocalPlayerIsInTopEntries() -> (LocalPlayerIsInTopEntries:bool))
   > DisplayTopEntries() {is=title K2Node_CallFunction}
   DisplayCenteredLeaderboardData(execute=@n58.else) {is=title K2Node_CallFunction}
   n59 = "Get (a copy)"(Array=@n50, Dimension 1=@n57)
   Branch(execute=@n57.LoopBody, Condition=Break Pilgrim Leaderboard Entry(PilgrimLeaderboardEntry=@n59) -> (bHasLocalPlayer=~, Percentile=~))
   > n60 = IsCurrentBandEntry(LeaderboardEntry=@n59)
   > n61 = Set LocalPlayerIsInTopEntries(LocalPlayerIsInTopEntries:bool=@n60) -> (Output_Get:bool)
   > Branch(Condition=@n61)
   > n62 = Set PlayerEntryIndex(PlayerEntryIndex=@n57) -> (Output_Get=0)
   Set ShouldRequestFriendsLeaderboard(execute=@n54.else, ShouldRequestFriendsLeaderboard:bool=false) -> (Output_Get:bool)
   > Call On Fallback to Global Leaderboard!(execute:Exec, self:WBP Pilgrim Quickplay Leaderboard Mini Object Reference) -> (then:Exec) {is=title K2Node_CallDelegate}
   > RequestTopEntries() {is=title K2Node_CallFunction}
   n63 = InitializeLeaderboard!() -> (OutputDelegate, then, NumPlayersInBand:int) {is=title K2Node_CustomEvent}
   > SetLeaderboardRequestData!(execute, self, NumPlayersInBand:int=@n63.NumPlayersInBand) -> (then) {is=title K2Node_CallFunction}
   > Branch(Condition=Get IsAllLeaderboardDataGathered() -> (IsAllLeaderboardDataGathered:bool))
   > RequestTopEntries()
   Event Construct()
   > Set PlayerEntryIndex(PlayerEntryIndex=-1) -> (Output_Get=0)
   Event Destruct()
   > Parent: Destruct()
   > n64 = Sequence()
   n65 = Get LeaderboardRequest() -> (LeaderboardRequest:Pilgrim Async Action Request Leaderboard Object Reference)
   > Is Valid(InputObject=@n65)
   > Cancel(self=@n65)
   > Set LeaderboardRequest()
   n66 = Get CenteredLeaderboardRequest()
   Is Valid(exec=@n64.then_1, InputObject=@n66)
   > Cancel(self=@n66)
   > Set CenteredLeaderboardRequest()
   UpdateScoreBeatenText() {is=title K2Node_CustomEvent}
  ]}
 Get Song Short Name For Leaderboard Request {nodes [
   Get Song Short Name For Leaderboard Request() {is=title K2Node_FunctionEntry}
   > n1 = GetPilgrimQuickplayPlayerStateComponent(PlayerState=Get PlayerState(self=GetOwningPlayer()))
   > n2 = Is Valid(InputObject=@n1)
   Return Node!(execute=@n2.Is Not Valid, SongShortName:Name=None)
   n3 = GetBandState(self=@n1)
   > Return Node(SongShortName=GetShortName!(self:Sparks Song Data Object Reference=GetCurrentSong!(self:Pilgrim Core Band State Component Object Reference=@n3) -> (ReturnValue:Pilgrim Song Metadata Object Reference)) -> (ReturnValue:Name=None))
  ]}
 GetScoreBeatenText {nodes [
   GetScoreBeatenText() {is=title K2Node_FunctionEntry}
   > Return Node!(execute, ScoreBeatenText:Text="LOCTEXT(\"You beat {PLAYER} and {NUM} others\")")
  ]}
 IsCurrentBandEntry {nodes [
   n1 = IsCurrentBandEntry!() -> (then, LeaderboardEntry:Pilgrim Leaderboard Entry Structure) {is=title K2Node_FunctionEntry}
   > n2 = Branch(Condition=Is Not Empty!(TargetArray:Array of Player State Object References=Get PlayersInBand() -> (PlayersInBand:Array of Player State Object References)) -> (ReturnValue:bool))
   n3 = Get PlayersInBand()
   n4 = Break Pilgrim Leaderboard Entry(PilgrimLeaderboardEntry=@n1) -> (bHasLocalPlayer=~, Percentile=~)
   > n5 = Branch(Condition="Equal (Integer)"(
     A=Length(TargetArray:Array of Player State Object References=@n3)
     B=Length!(TargetMap:Map of Strings to Strings=@n4.PlayerAccountIdToDisplayNameMap) -> (ReturnValue:int=0) {is=title K2Node_CallFunction}
    ))
   > n6 = For Each Loop with Break(Array:Array of Player State Object References=@n3, Break=@n9.else) -> (Array Element:Player State Object Reference)
   Return Node!(execute=[@n6.Completed, @n5.else, @n2.else], AllBandMembersMatch:bool=Get AllBandMembersMatchEntry() -> (AllBandMembersMatchEntry:bool))
   > n7 = GetDebugStringForUniqueId!(
    execute
    UniqueID:"Unique Net Id Structure (by ref)"=GetGameAccountId!(self:Fort Player Controller Object Reference=GetFortPlayerControllerFromActor!(Actor:Actor Object Reference=@n6) -> (ReturnValue:Fort Player Controller Object Reference)) -> (ReturnValue:Unique Net Id Structure)
   ) -> (then, ReturnValue:String)
   > n8 = Set AllBandMembersMatchEntry(AllBandMembersMatchEntry:bool=Contains!(
     TargetMap:Map of Strings to Strings=Break Pilgrim Leaderboard Entry(PilgrimLeaderboardEntry=@n1) -> (bHasLocalPlayer=~, Percentile=~).PlayerAccountIdToDisplayNameMap
     Key:"String (by ref)"=@n7
    ) -> (ReturnValue)) -> (Output_Get:bool)
   > n9 = Branch(Condition=@n8)
  ]}
 OnFallbackToGlobalLeaderboard {nodes [OnFallbackToGlobalLeaderboard() {is=title K2Node_FunctionEntry}]}
]
variables [
 CenteredLeaderboardPage {
  type=PilgrimLeaderboardPage
  default {
   bIsAllTimeHighScore=false
   bIsSpotlightCompetition=false
   leaderboardType=SoloGuitar
   page=0
   songShortName=None
   spotlightGroup=""
   totalPages=0
   entries []
  }
 }
 CenteredLeaderboardRequest {type=PilgrimAsyncAction_RequestLeaderboard, default=null}
 EntriesPlusMinusAroundPlayer {type=integer, default=1}
 FirstPageLeaderboardEntries {type="array<PilgrimLeaderboardEntry>", default []}
 IsAllLeaderboardDataGathered {type=boolean}
 LastDisplayIndex {type=integer}
 LeaderboardRequest {type=PilgrimAsyncAction_RequestLeaderboard, default=null}
 LeaderboardType {type=EPilgrimLeaderboardType, default=SoloGuitar}
 LocalPlayerIsInTopEntries {type=boolean}
 MaxRetries {type=integer, default=2}
 MinimumEntriesToDisplayFriendsLeaderboard {type=integer, default=2}
 NumTopEntriesToShow {type=integer, default=4}
 OnFallbackToGlobalLeaderboard {}
 PlayerEntryIndex {type=integer, default=-1}
 PlayerPageNumber {type=integer}
 PlayersInBand {type="array<PlayerState>", default []}
 RemainingCenteredRetries {type=integer}
 RemainingRetries {type=integer}
 RequestingPlayerController {type=FortPlayerController, default=null}
 RequestToDisplayCenteredDataHasBeenMade {type=boolean}
 ShouldRequestFriendsLeaderboard {type=boolean, default=true}
 SongShortName {type=string, default=None}
 SongTag {type=GameplayTag, default {tagName=None}}
 StartingDisplayIndex {type=integer}
 Total Pages {type=integer}
]
widgets {
 info {
  inheritedWidgetCount=0
  parentClass=&CommonUserWidget
  rootWidgetClass=&CommonVisibilitySwitcher
  widgetCount=10
 }
 widgets [
  ContentSwitcher {widgetClassPath=&CommonVisibilitySwitcher, bIsVariable=true, widget=ContentSwitcher}
  Content_LoadingEntries {
   widgetClassPath=&HorizontalBox
   bIsVariable=true
   parent=ContentSwitcher
   widget=Content_LoadingEntries
   slot {class=CommonVisibilitySwitcherSlot, properties {horizontalAlignment=HAlign_Center, verticalAlignment=VAlign_Center}}
  }
  ScaleBox_0 {
   widgetClassPath=&ScaleBox
   parent=Content_LoadingEntries
   widget=ScaleBox_0
   properties {stretchDirection=DownOnly}
   slot {class=HorizontalBoxSlot, properties {verticalAlignment=VAlign_Center, padding=<ltrb 25 0 25 0>}}
  }
  CommonTextBlock_0 {
   widgetClassPath=&CommonTextBlock
   parent=ScaleBox_0
   widget=CommonTextBlock_0
   properties {
    text=Loading
    textTransformPolicy=ToUpper
    font {
     bForceMonospaced=false
     bMaterialIsStencil=false
     fontMaterial=null
     fontObject=&BurbankSmall
     letterSpacing=0
     monospacedWidth=1
     size=18
     skewAmount=0
     typefaceFontName=Bold
     outlineSettings {
      bApplyOutlineToDropShadows=false
      bMiteredCorners=false
      bSeparateFillAlpha=false
      outlineMaterial=null
      outlineSize=0
      outlineColor=<rgba 0 0 0 1>
     }
    }
    shadowOffset=<0 0>
   }
   slot {class=ScaleBoxSlot}
  }
  Image_0 {
   widgetClassPath=&Image
   bIsVariable=true
   parent=Content_LoadingEntries
   widget=Image_0
   properties {
    visibility=SelfHitTestInvisible
    brush {
     bIsDynamicallyLoaded=false
     drawAs=Image
     imageType=NoImage
     mirroring=NoMirror
     resourceName=None
     resourceObject=/Game/UI/Foundation/Materials/Throbber_Material-Monochrome
     tiling=NoTile
     imageSize=<32 32>
     margin=<ltrb 0 0 0 0>
     outlineSettings {
      bUseBrushTransparency=false
      roundingType=HalfHeightRadius
      width=0
      color {colorUseRule=UseColor_Specified, specifiedColor=<rgba 0 0 0 0>}
      cornerRadii=<0 0 0 0>
     }
     tintColor {colorUseRule=UseColor_Specified, specifiedColor=<rgba 1 1 1 1>}
     uVRegion {bIsValid=false, max=<0 0>, min=<0 0>}
    }
   }
   slot {class=HorizontalBoxSlot, properties {horizontalAlignment=HAlign_Center, verticalAlignment=VAlign_Center}}
  }
  LeaderboardContainer {
   widgetClassPath=&VerticalBox
   bIsVariable=true
   parent=ContentSwitcher
   widget=LeaderboardContainer
   slot {class=CommonVisibilitySwitcherSlot}
  }
  LeaderboardEntries {
   widgetClassPath=&CommonListView
   bIsVariable=true
   parent=LeaderboardContainer
   widget=LeaderboardEntries
   properties {
    bIsFocusable=false
    entryWidgetClass=&/WBP_PilgrimQuickplay_LeaderboardEntry_Mini.WBP_PilgrimQuickplay_LeaderboardEntry_Mini_C
    numDesignerPreviewEntries=3
   }
   slot {class=VerticalBoxSlot}
  }
  ScoreBeatenTextBorder {
   widgetClassPath=&CommonBorder
   bIsVariable=true
   parent=LeaderboardContainer
   widget=ScoreBeatenTextBorder
   properties {
    verticalAlignment=VAlign_Center
    background {
     bIsDynamicallyLoaded=false
     drawAs=Box
     imageType=NoImage
     mirroring=NoMirror
     resourceName=None
     resourceObject=null
     tiling=NoTile
     imageSize=<32 32>
     margin=<ltrb 0 0 0 0>
     outlineSettings {
      bUseBrushTransparency=false
      roundingType=HalfHeightRadius
      width=0
      color {colorUseRule=UseColor_Specified, specifiedColor=<rgba 0 0 0 0>}
      cornerRadii=<0 0 0 0>
     }
     tintColor {colorUseRule=UseColor_Specified, specifiedColor=<rgba 1 1 1 1>}
     uVRegion {bIsValid=false, max=<0 0>, min=<0 0>}
    }
    brushColor=<rgba 1 0.9866 0 1>
    contentColorAndOpacity=<rgba 0 0 0 1>
    desiredSizeScale=<1 0>
   }
   slot {class=VerticalBoxSlot}
  }
  ScoreBeatenText {
   widgetClassPath=&CommonTextBlock
   bIsVariable=true
   parent=ScoreBeatenTextBorder
   widget=ScoreBeatenText
   properties {
    justification=Center
    scrollStyle=&TextScrollStyle-Base_C
    style=/FortUILibrary/Styles/Typography/Foundation/TextStyle_BurbankSmall_XS_Black.TextStyle_BurbankSmall_XS_Black_C
    text="You beat {PLAYER} and {NUM} others"
    colorAndOpacity {colorUseRule=UseColor_Specified, specifiedColor=<rgba 0 0 0 1>}
    font {
     bForceMonospaced=false
     bMaterialIsStencil=false
     fontMaterial=null
     fontObject=&BurbankSmall
     letterSpacing=0
     monospacedWidth=1
     size=13
     skewAmount=0
     typefaceFontName=Black
     outlineSettings {
      bApplyOutlineToDropShadows=false
      bMiteredCorners=false
      bSeparateFillAlpha=false
      outlineMaterial=null
      outlineSize=0
      outlineColor=<rgba 0 0 0 1>
     }
    }
    shadowOffset=<0 0>
   }
   slot {class=BorderSlot, properties {verticalAlignment=VAlign_Center}}
  }
  Text_NoLeaderboardEntries {
   widgetClassPath=&CommonTextBlock
   bIsVariable=true
   parent=ContentSwitcher
   widget=Text_NoLeaderboardEntries
   properties {
    autoWrapText=true
    justification=Center
    text="No Leaderboard Available \r\nTry Again Later"
    font {
     bForceMonospaced=false
     bMaterialIsStencil=false
     fontMaterial=null
     fontObject=&Burbank
     letterSpacing=0
     monospacedWidth=1
     size=24
     skewAmount=0
     typefaceFontName=Burbank-PlayerPower
     outlineSettings {
      bApplyOutlineToDropShadows=false
      bMiteredCorners=false
      bSeparateFillAlpha=false
      outlineMaterial=null
      outlineSize=0
      outlineColor=<rgba 0 0 0 1>
     }
    }
    shadowOffset=<0 0>
   }
   slot {class=CommonVisibilitySwitcherSlot, properties {verticalAlignment=VAlign_Center, padding=<ltrb 10 25 10 25>}}
  }
 ]
}
