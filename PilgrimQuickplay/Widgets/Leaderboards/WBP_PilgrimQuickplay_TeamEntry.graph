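# WBP_PilgrimQuickplay_TeamEntry is a leaderboard team entry widget that shows team usernames in one scrolling CommonTextBlock inside a SizeBox.
# SetText fills the label, and SetSelected plays the Selected color animation forward or in reverse to highlight or unhighlight it.
# Reset reverses that same animation at higher speed.
# A MainPlayer animation fades the text to white, but none of this widget's functions play it.
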WidgetBlueprint /PilgrimQuickplay/Widgets/Leaderboards/WBP_PilgrimQuickplay_TeamEntry
animations [
 Selected {
  binding_count=1
  length_seconds=0.3
  object=&self:Selected
  sequence {bindings [Text_Team {
     id=#1
     possessed_class=CommonTextBlock
     tracks [{
       type=MovieSceneColorTrack
       display_name=Color and Opacity
       sections [{
         type=MovieSceneColorSection
         end_frame=6
         start_frame=0
         channels [
          Color.R {type=MovieSceneScriptingFloatChannel, keys [[0,0.3916],[6,0.1651]]}
          Color.G {type=MovieSceneScriptingFloatChannel, keys [[0,0.4851],[6,0.9734]]}
          Color.B {type=MovieSceneScriptingFloatChannel, keys [[0,0.9387],[6,1]]}
          Color.A {type=MovieSceneScriptingFloatChannel, const [1,2]}
         ]
        }]
      }]
    }]}
 }
 MainPlayer {
  binding_count=1
  length_seconds=0.3
  object=&self:MainPlayer
  sequence {bindings [Text_Team {
     id=#1
     possessed_class=CommonTextBlock
     tracks [{
       type=MovieSceneColorTrack
       display_name=Color and Opacity
       sections [{
         type=MovieSceneColorSection
         end_frame=6
         start_frame=0
         channels [
          Color.R {type=MovieSceneScriptingFloatChannel, keys [[0,0.3916],[6,1]]}
          Color.G {type=MovieSceneScriptingFloatChannel, keys [[0,0.4851],[6,1]]}
          Color.B {type=MovieSceneScriptingFloatChannel, keys [[0,0.9387],[6,1]]}
          Color.A {type=MovieSceneScriptingFloatChannel, const [1,2]}
         ]
        }]
      }]
    }]}
 }
]
graphs [
 EventGraph {links [], nodes []}
 Reset {nodes [
   Reset() {is=title K2Node_FunctionEntry}
   > PlayAnimationReverse(InAnimation=Get Selected() -> (Selected:Widget Animation Object Reference), PlaybackSpeed=10)
  ]}
 SetSelected {nodes [
   n1 = SetSelected!() -> (then, IsSelected:bool=~) {is=title K2Node_FunctionEntry}
   > n2 = Branch(Condition=@n1)
   n3 = Get Selected()
   > PlayAnimationForward(InAnimation=@n3, PlaybackSpeed=4)
   PlayAnimationReverse(execute=@n2.else, InAnimation=@n3, PlaybackSpeed=3)
  ]}
 SetText {nodes [
   n1 = SetText!() -> (then, Usernames:Text) {is=title K2Node_FunctionEntry}
   > Set Text(Text=@n1, self=Get Text_Team() -> (Text_Team:Common Text Object Reference))
  ]}
]
variables []
widgets {
 info {inheritedWidgetCount=0, parentClass=&UserWidget, rootWidgetClass=&SizeBox, widgetCount=2}
 widgets [
  SizeBox_0 {widgetClassPath=&SizeBox, widget=SizeBox_0}
  Text_Team {
   widgetClassPath=&CommonTextBlock
   bIsVariable=true
   parent=SizeBox_0
   widget=Text_Team
   properties {
    scrollStyle=&TextScrollStyle-Base_C
    text=Username
    font {
     bForceMonospaced=false
     bMaterialIsStencil=false
     fontMaterial=null
     fontObject=&BurbankSmall
     letterSpacing=0
     monospacedWidth=1
     size=18
     skewAmount=0
     typefaceFontName=Black
     outlineSettings {
      bApplyOutlineToDropShadows=false
      bMiteredCorners=false
      bSeparateFillAlpha=false
      outlineMaterial=null
      outlineSize=0
      outlineColor=<rgba 0 0 0 1>
     }
    }
    shadowOffset=<0 0>
   }
   slot {class=SizeBoxSlot}
  }
 ]
}
