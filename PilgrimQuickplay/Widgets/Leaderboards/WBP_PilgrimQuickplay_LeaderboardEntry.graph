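# Leaderboard row widget for Pilgrim Quickplay showing rank, player names, season, stars, accuracy or full combo, difficulty and score.
# Setter events and functions fill the row, rebuild team-name entries, zebra-stripe by rank parity and highlight the current season via a curve table lookup.
# Selection plays the Selected animation and propagates to team entries; item reuse snaps it back.
# SetStars drives only the local star bar, leaving its solo/band branch disconnected.
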WidgetBlueprint /PilgrimQuickplay/Widgets/Leaderboards/WBP_PilgrimQuickplay_LeaderboardEntry
animations [
 Selected {
  binding_count=7
  length_seconds=0.3
  object=&self:Selected
  sequence {bindings [
    GlobalIcon {
     id=#1
     possessed_class=Image
     tracks [{
       type=MovieSceneColorTrack
       display_name=Color and Opacity
       sections [{
         type=MovieSceneColorSection
         end_frame=6
         start_frame=0
         channels [
          Color.R {type=MovieSceneScriptingFloatChannel, keys [[0,0.4125],[6,0.1651]]}
          Color.G {type=MovieSceneScriptingFloatChannel, keys [[0,0.5089],[6,0.9734]]}
          Color.B {type=MovieSceneScriptingFloatChannel, keys [[0,0.9647],[6,1]]}
          Color.A {type=MovieSceneScriptingFloatChannel, const [1,2]}
         ]
        }]
      }]
    }
    FriendIcon {
     id=#2
     possessed_class=Image
     tracks [{
       type=MovieSceneColorTrack
       display_name=Color and Opacity
       sections [{
         type=MovieSceneColorSection
         end_frame=6
         start_frame=0
         channels [
          Color.R {type=MovieSceneScriptingFloatChannel, keys [[0,0.4125],[6,0.1651]]}
          Color.G {type=MovieSceneScriptingFloatChannel, keys [[0,0.5089],[6,0.9734]]}
          Color.B {type=MovieSceneScriptingFloatChannel, keys [[0,0.9647],[6,1]]}
          Color.A {type=MovieSceneScriptingFloatChannel, const [1,2]}
         ]
        }]
      }]
    }
    Image_SelectedBG {
     id=#3
     possessed_class=Image
     tracks [{
       type=MovieSceneFloatTrack
       display_name=Render Opacity
       sections [{type=MovieSceneFloatSection, end_frame=6, start_frame=0, channels [None {type=MovieSceneScriptingFloatChannel, keys [[0,0],[6,1]]}]}]
      }]
    }
    Image_SelectedStroke {
     id=#4
     possessed_class=FortTouchRippleImage
     tracks [{
       type=MovieSceneFloatTrack
       display_name=Render Opacity
       sections [{type=MovieSceneFloatSection, end_frame=6, start_frame=0, channels [None {type=MovieSceneScriptingFloatChannel, keys [[0,0],[6,1]]}]}]
      }]
    }
    Text_Rank {
     id=#5
     possessed_class=CommonTextBlock
     tracks [{
       type=MovieSceneColorTrack
       display_name=Color and Opacity
       sections [{
         type=MovieSceneColorSection
         end_frame=6
         start_frame=0
         channels [
          Color.R {type=MovieSceneScriptingFloatChannel, keys [[0,0.3916],[6,0.1651]]}
          Color.G {type=MovieSceneScriptingFloatChannel, keys [[0,0.4852],[6,0.9734]]}
          Color.B {type=MovieSceneScriptingFloatChannel, keys [[0,0.9387],[6,1]]}
          Color.A {type=MovieSceneScriptingFloatChannel, const [1,2]}
         ]
        }]
      }]
    }
    Text_Score {
     id=#6
     possessed_class=CommonTextBlock
     tracks [{
       type=MovieSceneColorTrack
       display_name=Color and Opacity
       sections [{
         type=MovieSceneColorSection
         end_frame=6
         start_frame=0
         channels [
          Color.R {type=MovieSceneScriptingFloatChannel, keys [[0,0.4125],[6,0.1651]]}
          Color.G {type=MovieSceneScriptingFloatChannel, keys [[0,0.5089],[6,0.9734]]}
          Color.B {type=MovieSceneScriptingFloatChannel, keys [[0,0.9647],[6,1]]}
          Color.A {type=MovieSceneScriptingFloatChannel, const [1,2]}
         ]
        }]
      }]
    }
    Text_Accuracy {
     id=#7
     possessed_class=CommonTextBlock
     tracks [{
       type=MovieSceneColorTrack
       display_name=Color and Opacity
       sections [{
         type=MovieSceneColorSection
         end_frame=6
         start_frame=0
         channels [
          Color.R {type=MovieSceneScriptingFloatChannel, keys [[0,0.3916],[6,0.1651]]}
          Color.G {type=MovieSceneScriptingFloatChannel, keys [[0,0.4852],[6,0.9734]]}
          Color.B {type=MovieSceneScriptingFloatChannel, keys [[0,0.9387],[6,1]]}
          Color.A {type=MovieSceneScriptingFloatChannel, const [1,2]}
         ]
        }]
      }]
    }
   ]}
 }
 MainPlayer {
  binding_count=6
  length_seconds=0.3
  object=&self:MainPlayer
  sequence {bindings [
    GlobalIcon {
     id=#1
     possessed_class=Image
     tracks [{
       type=MovieSceneColorTrack
       display_name=Color and Opacity
       sections [{
         type=MovieSceneColorSection
         end_frame=6
         start_frame=0
         channels [
          Color.R {type=MovieSceneScriptingFloatChannel, keys [[0,0.3916],[6,1]]}
          Color.G {type=MovieSceneScriptingFloatChannel, keys [[0,0.4852],[6,1]]}
          Color.B {type=MovieSceneScriptingFloatChannel, keys [[0,0.9387],[6,1]]}
          Color.A {type=MovieSceneScriptingFloatChannel, const [1,2]}
         ]
        }]
      }]
    }
    FriendIcon {
     id=#2
     possessed_class=Image
     tracks [{
       type=MovieSceneColorTrack
       display_name=Color and Opacity
       sections [{
         type=MovieSceneColorSection
         end_frame=6
         start_frame=0
         channels [
          Color.R {type=MovieSceneScriptingFloatChannel, keys [[0,0.4125],[6,1]]}
          Color.G {type=MovieSceneScriptingFloatChannel, keys [[0,0.5089],[6,1]]}
          Color.B {type=MovieSceneScriptingFloatChannel, keys [[0,0.9647],[6,1]]}
          Color.A {type=MovieSceneScriptingFloatChannel, const [1,2]}
         ]
        }]
      }]
    }
    Text_Rank {
     id=#5
     possessed_class=CommonTextBlock
     tracks [{
       type=MovieSceneColorTrack
       display_name=Color and Opacity
       sections [{
         type=MovieSceneColorSection
         end_frame=6
         start_frame=0
         channels [
          Color.R {type=MovieSceneScriptingFloatChannel, keys [[0,0.3916],[6,1]]}
          Color.G {type=MovieSceneScriptingFloatChannel, keys [[0,0.4852],[6,1]]}
          Color.B {type=MovieSceneScriptingFloatChannel, keys [[0,0.9387],[6,1]]}
          Color.A {type=MovieSceneScriptingFloatChannel, const [1,2]}
         ]
        }]
      }]
    }
    Text_Score {
     id=#6
     possessed_class=CommonTextBlock
     tracks [{
       type=MovieSceneColorTrack
       display_name=Color and Opacity
       sections [{
         type=MovieSceneColorSection
         end_frame=6
         start_frame=0
         channels [
          Color.R {type=MovieSceneScriptingFloatChannel, keys [[0,0.4125],[6,1]]}
          Color.G {type=MovieSceneScriptingFloatChannel, keys [[0,0.5089],[6,1]]}
          Color.B {type=MovieSceneScriptingFloatChannel, keys [[0,0.9647],[6,1]]}
          Color.A {type=MovieSceneScriptingFloatChannel, const [1,2]}
         ]
        }]
      }]
    }
    Image_SelectedStroke {
     id=#8
     possessed_class=FortTouchRippleImage
     tracks [{
       type=MovieSceneFloatTrack
       display_name=Render Opacity
       sections [{type=MovieSceneFloatSection, end_frame=6, start_frame=0, channels [None {type=MovieSceneScriptingFloatChannel, const [0,2]}]}]
      }]
    }
    Text_Accuracy {
     id=#7
     possessed_class=CommonTextBlock
     tracks [{
       type=MovieSceneColorTrack
       display_name=Color and Opacity
       sections [{
         type=MovieSceneColorSection
         end_frame=6
         start_frame=0
         channels [
          Color.R {type=MovieSceneScriptingFloatChannel, keys [[0,0.3916],[6,1]]}
          Color.G {type=MovieSceneScriptingFloatChannel, keys [[0,0.4852],[6,1]]}
          Color.B {type=MovieSceneScriptingFloatChannel, keys [[0,0.9387],[6,1]]}
          Color.A {type=MovieSceneScriptingFloatChannel, const [1,2]}
         ]
        }]
      }]
    }
   ]}
 }
]
graphs [
 EventGraph {nodes [
   n1 = SetEntryNames!() -> (OutputDelegate, then, PlayerNames:Array of Strings) {is=title K2Node_CustomEvent}
   > Reset!(
    execute
    self:Dynamic Entry Box Object Reference=Get DynamicEntryBox_TeamEntries() -> (DynamicEntryBox_TeamEntries:Dynamic Entry Box Object Reference)
    bDeleteWidgets:bool=false
   ) -> (then)
   > n2 = For Each Loop(Array=@n1.PlayerNames)
   > n3 = Create Entry!(execute, self:Dynamic Entry Box Object Reference=Get DynamicEntryBox_TeamEntries()) -> (then, ReturnValue:User Widget Object Reference)
   > n4 = Cast To WBP_PilgrimQuickplay_TeamEntry!(execute, Object=@n3) -> (then, CastFailed, AsWBP Pilgrim Quickplay Team Entry:WBP Pilgrim Quickplay Team Entry Object Reference) {is=title K2Node_DynamicCast}
   > SetText!(
    execute
    self:WBP Pilgrim Quickplay Team Entry Object Reference=@n4
    Usernames:Text=Select(
     Option 0:Text="To Text (String)"(InString=@n2)
     Option 1:Text=Get UnknownCompetitor() -> (UnknownCompetitor:Text)
     Index=IsEmpty(InString=@n2)
    ) -> (ReturnValue:Text)
   ) -> (then)
   > Reset!(execute, self:WBP Pilgrim Quickplay Team Entry Object Reference=@n4) -> (then)
   n5 = SetAccuracy!() -> (OutputDelegate, then, Accuracy:double=~) {is=title K2Node_CustomEvent}
   > "SetText (Text)"(
    self=Get Text_Accuracy() -> (Text_Accuracy:Common Text Object Reference)
    InText=Format Text!(Format="LOCTEXT(\"{Accuracy}%\")", Accuracy:double=@n5.Accuracy) -> (Result)
   )
   n6 = SetScore!() -> (OutputDelegate, then, Score:int) {is=title K2Node_CustomEvent}
   > "SetText (Text)"(self=Get Text_Score() -> (Text_Score:Common Text Object Reference), InText="To Text (Integer)"(Value=@n6.Score))
   n7 = SetDifficulty!() -> (OutputDelegate, then, Difficulty:EPilgrimSongDifficulty Enum=DifficultyEasy) {is=title K2Node_CustomEvent}
   > "SetText (Text)"(
    self=Get CommonTextBlock-Difficulty!(self:WBP Pilgrim Results Difficulty Object Reference=Get WBP_Pilgrim_Results_Difficulty() -> (WBP_Pilgrim_Results_Difficulty:WBP Pilgrim Results Difficulty Object Reference)) -> (CommonTextBlock-Difficulty:Common Text Object Reference)
    InText=Select!(
     DifficultyEasy:"Text (by ref)"=MakeLiteralText(Value="LOCTEXT(\"E\")")
     DifficultyMedium:"Text (by ref)"=MakeLiteralText(Value="LOCTEXT(\"M\")")
     DifficultyHard:"Text (by ref)"=MakeLiteralText(Value="LOCTEXT(\"H\")")
     DifficultyExpert:"Text (by ref)"=MakeLiteralText(Value="LOCTEXT(\"X\")")
     None:"Text (by ref)"=MakeLiteralText(Value="LOCTEXT(\"N\")")
     Index:EPilgrimSongDifficulty Enum=@n7.Difficulty
    ) -> (ReturnValue:"Text (by ref)")
   )
   n8 = SetStarCount!() -> (OutputDelegate, then, NumStars:int) {is=title K2Node_CustomEvent}
   > SetStars!(execute, self, NumStars:int=@n8.NumStars) -> (then)
   n9 = SetFullCombo!() -> (OutputDelegate, then, Full Combo:bool=~) {is=title K2Node_CustomEvent}
   > SetActiveWidgetIndex(
    self:Common Visibility Switcher Object Reference=Get FullCombo_Switcher() -> (FullCombo_Switcher:Common Visibility Switcher Object Reference)
    Index="To Integer (Boolean)"!(InBool:bool=@n9.Full Combo) -> (ReturnValue:int=0)
   )
   Event On Entry Released() {is=title K2Node_Event}
   > SetFirstTimeState()
   Event Construct()
   > SetIsListItemClipping!(execute, self, IsListItem:bool=false) -> (then)
   Event OnListItemObjectSet()
   > SetFirstTimeState()
   n10 = Event On Item Selection Changed!() -> (OutputDelegate, then, bIsSelected:bool) {is=title K2Node_Event}
   n11 = Get IsFirstTime() -> (IsFirstTime:bool)
   n12 = Select!(Option 0:bool=false, Option 1:bool=@n10.bIsSelected, Index:bool=OR Boolean(A=@n10.bIsSelected, B=AND Boolean(A=@n10.bIsSelected, B=@n11))) -> (ReturnValue:bool)
   > n13 = Branch(Condition=@n12)
   > Set IsFirstTime(IsFirstTime:bool=false) -> (Output_Get:bool)
   n14 = Get Selected() -> (Selected:Widget Animation Object Reference)
   > n15 = PlayAnimationForward(InAnimation=@n14, PlaybackSpeed=4)
   n16 = For Each Loop(
    Exec=[@n15.then, @n18.then]
    Array:Array of User Widget Object References=GetAllEntries!(self:Dynamic Entry Box Base Object Reference=Get DynamicEntryBox_TeamEntries()) -> (ReturnValue:Array of User Widget Object References)
   ) -> (Array Element:User Widget Object Reference)
   > n17 = Cast To WBP_PilgrimQuickplay_TeamEntry(Object=@n16)
   > SetSelected!(execute, self:WBP Pilgrim Quickplay Team Entry Object Reference=@n17, IsSelected:bool=@n12) -> (then)
   n18 = PlayAnimationReverse(execute=@n13.else, InAnimation=@n14, PlaybackSpeed=3)
   Reroute Node(InputPin:Wildcard) -> (OutputPin:Wildcard)
  ]}
 SetCurrentUserBgVisibility {nodes [
   n1 = SetCurrentUserBgVisibility!() -> (then, IsCurrentUser:bool=~) {is=title K2Node_FunctionEntry}
   > SetVisibility(
    self=Get CurrentUserBG() -> (CurrentUserBG:Border Object Reference)
    InVisibility=Select(Option 0:ESlateVisibility Enum=Hidden, Option 1:ESlateVisibility Enum=SelfHitTestInvisible, Index=@n1) -> (ReturnValue:ESlateVisibility Enum)
   )
  ]}
 SetFirstTimeState {nodes [
   SetFirstTimeState() {is=title K2Node_FunctionEntry}
   > Set IsFirstTime(IsFirstTime=true)
   > PlayAnimationReverse(InAnimation=Get Selected(), PlaybackSpeed=100)
  ]}
 SetIsListItemClipping {nodes [
   n1 = SetIsListItemClipping!() -> (then, IsListItem:bool=~) {is=title K2Node_FunctionEntry}
   > Branch(Condition=@n1)
   > SetClipping!(execute, self:Widget Object Reference, InClipping:EWidgetClipping Enum=ClipToBounds) -> (then)
  ]}
 SetIsSoloEntry {nodes [
   n1 = SetIsSoloEntry!() -> (then, InputPin:bool=~) {is=title K2Node_FunctionEntry}
   > Set IsSoloEntry(IsSoloEntry:bool=@n1) -> (Output_Get:bool)
   > SetVisibility(
    self=[Get OverlayAccuracy() -> (OverlayAccuracy:Overlay Object Reference), Get WBP_Pilgrim_Results_Difficulty()]
    InVisibility=Select(Option 0=Collapsed, Option 1=SelfHitTestInvisible, Index=@n1)
   )
  ]}
 SetPlayerRank {nodes [
   n1 = SetPlayerRank!() -> (then, Friend or Player Rank:int) {is=title K2Node_FunctionEntry}
   > "SetText (Text)"(self=Get Text_Rank() -> (Text_Rank:Common Text Object Reference), InText="To Text (Integer)"(Value=@n1))
   n2 = "% (Integer)"(A=int + int(A=@n1), B=2) -> (ReturnValue=0)
   > SetBrushColor(
    self=Get CommonBorder_Zebra() -> (CommonBorder_Zebra:Common Border Object Reference)
    InBrushColor=Select(
     Option 0:Linear Color Structure=(R=0,G=0,B=0,A=0)
     Option 1:Linear Color Structure=Make Linear Color!(R:float=0.01, G:float=0, B:float=0.15, A:float=0.2) -> (LinearColor:Linear Color Structure) {is=title K2Node_MakeStruct}
     Index="Equal (Integer)"(A=@n2, B=0) {is=title K2Node_CallFunction}
    ) -> (ReturnValue:Linear Color Structure)
   )
  ]}
 SetRankIcon {nodes [
   n1 = SetRankIcon!() -> (then, FriendRank:int) {is=title K2Node_FunctionEntry}
   > n2 = Branch(Condition="integer <= integer"(A=@n1))
   n3 = Get RankIconSwitcher() -> (RankIconSwitcher:Common Visibility Switcher Object Reference)
   > SetActiveWidgetIndex(self=@n3, Index=0)
   SetActiveWidgetIndex(execute=@n2.else, self=@n3, Index=1)
  ]}
 SetSeasonText {nodes [
   n1 = SetSeasonText!() -> (then, SeasonText:int) {is=title K2Node_FunctionEntry}
   > n2 = EvaluateCurveTableRow!(
    execute
    CurveTableRowHandle:Curve Table Row Handle Structure=Make Curve Table Row Handle!(CurveTable:Curve Table Object Reference=/PilgrimQuickplay/DataTables/PilgrimQuickplayGameData, RowName:Name=SeasonData.SeasonNumber) -> (CurveTableRowHandle:Curve Table Row Handle Structure) {is=title K2Node_MakeStruct}
    InXY:float=0
    ContextString:String
   ) -> (then, OutXY:float, ReturnValue:bool)
   > n3 = Branch(Condition="Equal (Float)"(A:int=@n1, B=@n2))
   n4 = Get WBP_PilgrimQuickplay_SeasonDisplay() -> (WBP_PilgrimQuickplay_SeasonDisplay:WBP Pilgrim Quickplay Season Display Object Reference)
   > n5 = SetCurrentSeasonBorder(self:WBP Pilgrim Quickplay Season Display Object Reference=@n4)
   Set Text(
    execute=[@n5.then, @n6.then]
    Text=Format Text!(Format="LOCTEXT(\"S{SeasonNumber}\")", SeasonNumber:int=@n1) -> (Result)
    self=Get SeasonText!(self:WBP Pilgrim Quickplay Season Display Object Reference=Get WBP_PilgrimQuickplay_SeasonDisplay()) -> (SeasonText:Common Text Object Reference)
   )
   n6 = RevertSeasonBorder(execute=@n3.else, self:WBP Pilgrim Quickplay Season Display Object Reference=@n4)
  ]}
 SetStars {nodes [
   SetActiveWidgetIndex(self=Get StarsVisibilitySwitcher() -> (StarsVisibilitySwitcher:Common Visibility Switcher Object Reference), Index=1)
   n1 = Get IsSoloEntry() -> (IsSoloEntry:bool)
   > n2 = Branch(Condition=@n1)
   n3 = Get WBP_PilgrimQuickplay_LocalStarBar_2() -> (WBP_PilgrimQuickplay_LocalStarBar_2:WBP Pilgrim Quickplay Local Star Bar Object Reference)
   > ResetCompletedVisuals(self:WBP Pilgrim Quickplay Local Star Bar Object Reference=@n3)
   n4 = Reroute Node(InputPin:Array of EPilgrimTrackType Enums) -> (OutputPin:Array of EPilgrimTrackType Enums)
   > SetupStarsResults_DEPRECATED!(
    execute
    self:WBP Pilgrim Quickplay Local Star Bar Object Reference=@n3
    Instrument:EPilgrimTrackType Enum="Get (a ref)"(Array:Array of EPilgrimTrackType Enums=@n4, Dimension 1=Last Index!(TargetArray:Array of EPilgrimTrackType Enums=@n4) -> (ReturnValue=0)) -> (Output:"EPilgrimTrackType Enum (by ref)")
    StarCount:int=Reroute Node(InputPin:int) -> (OutputPin:int)
   ) -> (then)
   n5 = Get WBP_PilgrimQuickplay_BandStarBar() -> (WBP_PilgrimQuickplay_BandStarBar:WBP Pilgrim Quickplay Band Star Bar Object Reference)
   ResetCompletedVisuals(execute=@n2.else, self:WBP Pilgrim Quickplay Band Star Bar Object Reference=@n5)
   > SetupStarsResults!(execute, self:WBP Pilgrim Quickplay Band Star Bar Object Reference=@n5, StarCount:int=Reroute Node()) -> (then)
   n6 = SetStars!() -> (then, NumStars:int) {is=title K2Node_FunctionEntry}
   n7 = Get WBP_PilgrimQuickplay_LocalStarBar_2()
   > ResetCompletedVisuals(self:WBP Pilgrim Quickplay Local Star Bar Object Reference=@n7)
   > SetupStarCount!(execute, self:WBP Pilgrim Quickplay Local Star Bar Object Reference=@n7, StarCount:int=@n6) -> (then)
   "To Integer (Boolean)"(InBool=@n1) -> (ReturnValue=0)
  ]}
]
variables [
 IsFirstTime {type=boolean, default=true}
 IsSoloEntry {type=boolean}
 UnknownCompetitor {type=string, default="[Unknown Competitor]"}
]
widgets {
 info {inheritedWidgetCount=0, parentClass=&UserWidget, rootWidgetClass=&Overlay, widgetCount=31}
 widgets [
  Overlay_31 {widgetClassPath=&Overlay, widget=Overlay_31}
  EvenBG {
   widgetClassPath=&Image
   bIsVariable=true
   parent=Overlay_31
   widget=EvenBG
   properties {colorAndOpacity=<rgba 1 1 1 0>}
   slot {class=OverlaySlot, properties {horizontalAlignment=HAlign_Fill, verticalAlignment=VAlign_Fill}}
  }
  CommonBorder_Zebra {
   widgetClassPath=&CommonBorder
   bIsVariable=true
   parent=Overlay_31
   widget=CommonBorder_Zebra
   properties {
    background {
     bIsDynamicallyLoaded=false
     drawAs=Image
     imageType=NoImage
     mirroring=NoMirror
     resourceName=None
     resourceObject=/PilgrimCore/Widgets/Materials/M_UI_Pilgrim_LeaderBoard_Backing
     tiling=NoTile
     imageSize=<32 32>
     margin=<ltrb 0 0 0 0>
     outlineSettings {
      bUseBrushTransparency=false
      roundingType=HalfHeightRadius
      width=0
      color {colorUseRule=UseColor_Specified, specifiedColor=<rgba 0 0 0 0>}
      cornerRadii=<0 0 0 0>
     }
     tintColor {colorUseRule=UseColor_Specified, specifiedColor=<rgba 1 1 1 1>}
     uVRegion {bIsValid=false, max=<0 0>, min=<0 0>}
    }
    padding=<ltrb 16 0 0 0>
   }
   slot {class=OverlaySlot, properties {horizontalAlignment=HAlign_Fill, verticalAlignment=VAlign_Fill}}
  }
  CurrentUserBG {
   widgetClassPath=&Border
   bIsVariable=true
   parent=Overlay_31
   widget=CurrentUserBG
   properties {
    renderOpacity=0.75
    visibility=Hidden
    background {
     bIsDynamicallyLoaded=false
     drawAs=RoundedBox
     imageType=FullColor
     mirroring=NoMirror
     resourceName=None
     resourceObject=/PilgrimQuickplay/Widgets/Leaderboards/Materials/MI_Leaderboard_CurrentUser_Purple
     tiling=NoTile
     imageSize=<32 32>
     margin=<ltrb 0 0 0 0>
     outlineSettings {
      bUseBrushTransparency=false
      roundingType=FixedRadius
      width=0
      color {colorUseRule=UseColor_Specified, specifiedColor=<rgba 0 0 0 0>}
      cornerRadii=<16 16 16 16>
     }
     tintColor {colorUseRule=UseColor_Specified, specifiedColor=<rgba 1 1 1 1>}
     uVRegion {bIsValid=false, max=<0 0>, min=<0 0>}
    }
    padding=<ltrb 0 0 0 0>
   }
   slot {class=OverlaySlot, properties {horizontalAlignment=HAlign_Fill, verticalAlignment=VAlign_Fill}}
  }
  Image_SelectedBG {
   widgetClassPath=&Image
   bIsVariable=true
   parent=Overlay_31
   widget=Image_SelectedBG
   properties {
    renderOpacity=0
    brush {
     bIsDynamicallyLoaded=false
     drawAs=RoundedBox
     imageType=FullColor
     mirroring=NoMirror
     resourceName=None
     resourceObject=/PilgrimQuickplay/Widgets/Leaderboards/Materials/MI_Leaderboard_SelectedUser
     tiling=NoTile
     imageSize=<256 64>
     margin=<ltrb 0.4 0.2 0.4 0.2>
     outlineSettings {
      bUseBrushTransparency=false
      roundingType=FixedRadius
      width=0
      color {colorUseRule=UseColor_Specified, specifiedColor=<rgba 0 0 0 0>}
      cornerRadii=<16 16 16 16>
     }
     tintColor {colorUseRule=UseColor_Specified, specifiedColor=<rgba 1 1 1 1>}
     uVRegion {bIsValid=false, max=<0 0>, min=<0 0>}
    }
   }
   slot {class=OverlaySlot, properties {horizontalAlignment=HAlign_Fill, verticalAlignment=VAlign_Fill}}
  }
  Image_SelectedStroke {
   widgetClassPath=/Script/FortniteUI.FortTouchRippleImage
   bIsVariable=true
   parent=Overlay_31
   widget=Image_SelectedStroke
   properties {brush {
     bIsDynamicallyLoaded=false
     drawAs=Image
     imageType=FullColor
     mirroring=NoMirror
     resourceName=None
     resourceObject=/PilgrimCore/Widgets/Materials/MI_UI_Pilgrim_Rounded_Outline_SelectedLeadeboards
     tiling=NoTile
     imageSize=<32 32>
     margin=<ltrb 0 0 0 0>
     outlineSettings {
      bUseBrushTransparency=false
      roundingType=HalfHeightRadius
      width=0
      color {colorUseRule=UseColor_Specified, specifiedColor=<rgba 0 0 0 0>}
      cornerRadii=<0 0 0 0>
     }
     tintColor {colorUseRule=UseColor_Specified, specifiedColor=<rgba 1 1 1 1>}
     uVRegion {bIsValid=false, max=<0 0>, min=<0 0>}
    }}
   slot {class=OverlaySlot, properties {horizontalAlignment=HAlign_Fill, verticalAlignment=VAlign_Fill}}
  }
  HorizontalBox_0 {
   widgetClassPath=&HorizontalBox
   parent=Overlay_31
   widget=HorizontalBox_0
   slot {class=OverlaySlot, properties {horizontalAlignment=HAlign_Fill, verticalAlignment=VAlign_Fill, padding=<ltrb 16 0 0 0>}}
  }
  RankIconSwitcher {
   widgetClassPath=&CommonVisibilitySwitcher
   bIsVariable=true
   parent=HorizontalBox_0
   widget=RankIconSwitcher
   slot {class=HorizontalBoxSlot, properties {verticalAlignment=VAlign_Center, padding=<ltrb 0 5 10 5>}}
  }
  GlobalIcon {
   widgetClassPath=&Image
   bIsVariable=true
   parent=RankIconSwitcher
   widget=GlobalIcon
   properties {
    visibility=SelfHitTestInvisible
    brush {
     bIsDynamicallyLoaded=false
     drawAs=Image
     imageType=FullColor
     mirroring=NoMirror
     resourceName=None
     resourceObject=/Game/UI/Friends_UI/Social/Globe_PlatformIcon_32x
     tiling=NoTile
     imageSize=<32 32>
     margin=<ltrb 0 0 0 0>
     outlineSettings {
      bUseBrushTransparency=false
      roundingType=HalfHeightRadius
      width=0
      color {colorUseRule=UseColor_Specified, specifiedColor=<rgba 0 0 0 0>}
      cornerRadii=<0 0 0 0>
     }
     tintColor {colorUseRule=UseColor_Specified, specifiedColor=<rgba 1 1 1 1>}
     uVRegion {bIsValid=false, max=<0 0>, min=<0 0>}
    }
    colorAndOpacity=<rgba 0.3916 0.4852 0.9387 1>
   }
   slot {class=CommonVisibilitySwitcherSlot}
  }
  FriendIcon {
   widgetClassPath=&Image
   bIsVariable=true
   parent=RankIconSwitcher
   widget=FriendIcon
   properties {
    visibility=SelfHitTestInvisible
    brush {
     bIsDynamicallyLoaded=false
     drawAs=Image
     imageType=FullColor
     mirroring=NoMirror
     resourceName=None
     resourceObject=/Game/UI/Foundation/Shell/Textures/T-Icon-Friends-L
     tiling=NoTile
     imageSize=<32 32>
     margin=<ltrb 0 0 0 0>
     outlineSettings {
      bUseBrushTransparency=false
      roundingType=HalfHeightRadius
      width=0
      color {colorUseRule=UseColor_Specified, specifiedColor=<rgba 0 0 0 0>}
      cornerRadii=<0 0 0 0>
     }
     tintColor {colorUseRule=UseColor_Specified, specifiedColor=<rgba 1 1 1 1>}
     uVRegion {bIsValid=false, max=<0 0>, min=<0 0>}
    }
    colorAndOpacity=<rgba 0.1651 0.9734 1 1>
   }
   slot {class=CommonVisibilitySwitcherSlot}
  }
  Rank {widgetClassPath=&Overlay, parent=HorizontalBox_0, widget=Rank, slot {class=HorizontalBoxSlot}}
  Text_Rank {
   widgetClassPath=&CommonTextBlock
   bIsVariable=true
   parent=Rank
   widget=Text_Rank
   properties {
    text=1
    colorAndOpacity {colorUseRule=UseColor_Specified, specifiedColor=<rgba 0.1651 0.9734 1 1>}
    font {
     bForceMonospaced=true
     bMaterialIsStencil=false
     fontMaterial=null
     fontObject=&HeadingNow
     letterSpacing=0
     monospacedWidth=1
     size=24
     skewAmount=0
     typefaceFontName=75 Medium
     outlineSettings {
      bApplyOutlineToDropShadows=false
      bMiteredCorners=false
      bSeparateFillAlpha=false
      outlineMaterial=null
      outlineSize=0
      outlineColor=<rgba 0 0 0 1>
     }
    }
    shadowOffset=<0 0>
   }
   slot {class=OverlaySlot, properties {horizontalAlignment=HAlign_Center, verticalAlignment=VAlign_Center}}
  }
  Overlay_0 {
   widgetClassPath=&Overlay
   parent=HorizontalBox_0
   widget=Overlay_0
   slot {class=HorizontalBoxSlot, properties {size {sizeRule=Fill, value=1}}}
  }
  HorizontalBoxTabLayout {
   widgetClassPath=&HorizontalBox
   parent=Overlay_0
   widget=HorizontalBoxTabLayout
   slot {class=OverlaySlot, properties {horizontalAlignment=HAlign_Fill, verticalAlignment=VAlign_Center, padding=<ltrb 16 0 0 0>}}
  }
  DynamicEntryBox_TeamEntries {
   widgetClassPath=&DynamicEntryBox
   bIsVariable=true
   parent=HorizontalBoxTabLayout
   widget=DynamicEntryBox_TeamEntries
   properties {
    entryBoxType=Vertical
    entryWidgetClass=&/WBP_PilgrimQuickplay_TeamEntry.WBP_PilgrimQuickplay_TeamEntry_C
   }
   slot {class=HorizontalBoxSlot, properties {verticalAlignment=VAlign_Center, padding=<ltrb 0 0 5 0>, size {sizeRule=Fill, value=1}}}
  }
  WBP_PilgrimQuickplay_SeasonDisplay {
   widgetClassPath=/PilgrimQuickplay/Widgets/WBP_PilgrimQuickplay_SeasonDisplay
   bIsVariable=true
   parent=HorizontalBoxTabLayout
   widget=WBP_PilgrimQuickplay_SeasonDisplay
   slot {class=HorizontalBoxSlot, properties {verticalAlignment=VAlign_Center, padding=<ltrb 0 0 12 0>}}
  }
  StarsVisibilitySwitcher {
   widgetClassPath=&CommonVisibilitySwitcher
   bIsVariable=true
   parent=HorizontalBoxTabLayout
   widget=StarsVisibilitySwitcher
   properties {activeWidgetIndex=1}
   slot {
    class=HorizontalBoxSlot
    properties {horizontalAlignment=HAlign_Right, verticalAlignment=VAlign_Center, size {sizeRule=Automatic, value=0.2}}
   }
  }
  WBP_PilgrimQuickplay_BandStarBar {
   widgetClassPath=/PilgrimQuickplay/Widgets/WBP_PilgrimQuickplay_BandStarBar
   bIsVariable=true
   parent=StarsVisibilitySwitcher
   widget=WBP_PilgrimQuickplay_BandStarBar
   properties {in User Specified Scale=0.65, scale Stretch=UserSpecified, shouldStarsAnimateComplete=false}
   slot {class=CommonVisibilitySwitcherSlot}
  }
  WBP_PilgrimQuickplay_LocalStarBar_2 {
   widgetClassPath=/PilgrimQuickplay/Widgets/WBP_PilgrimQuickplay_LocalStarBar
   bIsVariable=true
   parent=StarsVisibilitySwitcher
   widget=WBP_PilgrimQuickplay_LocalStarBar_2
   properties {in User Specified Scale=0.65, scale Stretch=UserSpecified}
   slot {class=CommonVisibilitySwitcherSlot}
  }
  HorizontalBox_2 {
   widgetClassPath=&HorizontalBox
   parent=HorizontalBoxTabLayout
   widget=HorizontalBox_2
   slot {class=HorizontalBoxSlot, properties {horizontalAlignment=HAlign_Right, verticalAlignment=VAlign_Center}}
  }
  OverlayAccuracy {
   widgetClassPath=&Overlay
   bIsVariable=true
   parent=HorizontalBox_2
   widget=OverlayAccuracy
   slot {class=HorizontalBoxSlot, properties {verticalAlignment=VAlign_Center, padding=<ltrb 12 0 12 0>}}
  }
  Spacer_185 {
   widgetClassPath=&Spacer
   parent=OverlayAccuracy
   widget=Spacer_185
   properties {size=<78 1>}
   slot {class=OverlaySlot, properties {verticalAlignment=VAlign_Bottom}}
  }
  FullCombo_Switcher {
   widgetClassPath=&CommonVisibilitySwitcher
   bIsVariable=true
   parent=OverlayAccuracy
   widget=FullCombo_Switcher
   slot {class=OverlaySlot, properties {horizontalAlignment=HAlign_Center}}
  }
  Text_Accuracy {
   widgetClassPath=&CommonTextBlock
   bIsVariable=true
   parent=FullCombo_Switcher
   widget=Text_Accuracy
   properties {
    text=100%
    colorAndOpacity {colorUseRule=UseColor_Specified, specifiedColor=<rgba 0.1651 0.9734 1 1>}
    font {
     bForceMonospaced=false
     bMaterialIsStencil=false
     fontMaterial=null
     fontObject=&BurbankSmall
     letterSpacing=0
     monospacedWidth=1
     size=18
     skewAmount=0
     typefaceFontName=Black
     outlineSettings {
      bApplyOutlineToDropShadows=false
      bMiteredCorners=false
      bSeparateFillAlpha=false
      outlineMaterial=null
      outlineSize=0
      outlineColor=<rgba 0 0 0 1>
     }
    }
    shadowOffset=<0 0>
   }
   slot {class=CommonVisibilitySwitcherSlot}
  }
  WBP_PilgrimQuickplay_FullComboIndicator {
   widgetClassPath=/PilgrimQuickplay/Widgets/WBP_PilgrimQuickplay_FullComboIndicator
   bIsVariable=true
   parent=FullCombo_Switcher
   widget=WBP_PilgrimQuickplay_FullComboIndicator
   slot {class=CommonVisibilitySwitcherSlot}
  }
  OverlayDifficulty {
   widgetClassPath=&Overlay
   parent=HorizontalBox_2
   widget=OverlayDifficulty
   slot {class=HorizontalBoxSlot, properties {verticalAlignment=VAlign_Center, padding=<ltrb 6 0 3 0>}}
  }
  Spacer_147 {
   widgetClassPath=&Spacer
   parent=OverlayDifficulty
   widget=Spacer_147
   properties {size=<34 1>}
   slot {class=OverlaySlot, properties {horizontalAlignment=HAlign_Fill, verticalAlignment=VAlign_Fill}}
  }
  WBP_Pilgrim_Results_Difficulty {
   widgetClassPath=/PilgrimQuickplay/Widgets/Results/WBP_Pilgrim_Results_Difficulty
   bIsVariable=true
   parent=OverlayDifficulty
   widget=WBP_Pilgrim_Results_Difficulty
   slot {class=OverlaySlot}
  }
  OverlayScore {
   widgetClassPath=&Overlay
   parent=HorizontalBox_2
   widget=OverlayScore
   slot {class=HorizontalBoxSlot, properties {verticalAlignment=VAlign_Center, padding=<ltrb 6 0 0 0>}}
  }
  Spacer_82 {widgetClassPath=&Spacer, parent=OverlayScore, widget=Spacer_82, properties {size=<155 1>}, slot {class=OverlaySlot}}
  Text_Score {
   widgetClassPath=&CommonTextBlock
   bIsVariable=true
   parent=OverlayScore
   widget=Text_Score
   properties {
    justification=Center
    text="9,999,999"
    font {
     bForceMonospaced=true
     bMaterialIsStencil=false
     fontMaterial=null
     fontObject=&HeadingNow
     letterSpacing=10
     monospacedWidth=0.85
     size=18
     skewAmount=0
     typefaceFontName=74 Regular
     outlineSettings {
      bApplyOutlineToDropShadows=false
      bMiteredCorners=false
      bSeparateFillAlpha=false
      outlineMaterial=null
      outlineSize=0
      outlineColor=<rgba 0 0 0 1>
     }
    }
    shadowOffset=<0 0>
   }
   slot {class=OverlaySlot, properties {horizontalAlignment=HAlign_Right, padding=<ltrb 0 0 16 0>}}
  }
 ]
}
