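# Leaderboard navigation bar for Pilgrim Quickplay, a child of the UIKit top nav tabs widget with no widgets of its own.
# PreConstruct applies an input-action override, defaulting to Custom.
# UpdateInputActions switches on it, either binding custom next/previous tab actions or deferring to the parent.
# Construct collapses the Next prompt and spacer; input-method changes collapse both spacers for mouse and keyboard, otherwise show the Previous prompt and spacer.
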WidgetBlueprint /PilgrimQuickplay/Widgets/Leaderboards/WBP_PilgrimQuickplay_Leaderboard_NavBar
graphs [
 EventGraph {nodes [
   n1 = Event PreConstruct()
   > Parent: PreConstruct(IsDesignTime=@n1.IsDesignTime)
   > SetInputActionsType!(
    execute
    self:WBP UIKit Tabs Base Object Reference
    InputActions:E_UI_TabsInputActionType Enum=Get Input Actions Override() -> (Input Actions Override:E_UI_TabsInputActionType Enum=NewEnumerator2)
    InputActionsWithMouseScroll:bool=false
   ) -> (then)
   Event Construct()
   > Parent: Construct()
   > SetVisibility(
    self=[Get Next_Prompt() -> (Next_Prompt:Grid Panel Object Reference), Get Next_Spacer() -> (Next_Spacer:Spacer Object Reference)]
    InVisibility=Collapsed
   )
   n2 = Event Event On Input Method Changed!() -> (OutputDelegate, then, bNewInputType:ECommonInputType Enum) {is=title K2Node_Event}
   > n3 = Branch(Condition="Equal (Enum)"(A=@n2.bNewInputType, B=MouseAndKeyboard))
   > SetVisibility(self=[Get Next_Spacer(), Get Prev_Spacer() -> (Prev_Spacer:Spacer Object Reference)], InVisibility=Collapsed)
   SetVisibility(
    execute=@n3.else
    self=[Get Prev_Prompt() -> (Prev_Prompt:Grid Panel Object Reference), Get Prev_Spacer()]
    InVisibility=SelfHitTestInvisible
   )
  ]}
 SetInputActionsType {nodes [
   n1 = SetInputActionsType!() -> (then, InputActions:E_UI_TabsInputActionType Enum=NewEnumerator2, InputActionsWithMouseScroll:bool) {is=title K2Node_FunctionEntry}
   > n2 = Set Input Actions Override(Input Actions Override:E_UI_TabsInputActionType Enum=@n1) -> (Output_Get:E_UI_TabsInputActionType Enum=NewEnumerator2)
   > Parent: SetInputActionsType!(execute, InputActions:E_UI_TabsInputActionType Enum=@n2, InputActionsWithMouseScroll:bool=@n1.InputActionsWithMouseScroll) -> (then) {is=title K2Node_CallParentFunction}
  ]}
 UpdateInputActions {nodes [
   UpdateInputActions() {is=title K2Node_FunctionEntry}
   > n1 = Switch on E_UI_TabsInputActionType!(execute, Selection:E_UI_TabsInputActionType Enum=Get Input Actions Override() -> (Input Actions Override=NewEnumerator2)) -> (NewEnumerator2:Exec, NewEnumerator0:Exec, NewEnumerator1:Exec, NewEnumerator3:Exec) {is=title K2Node_SwitchEnum}
   Parent: UpdateInputActions(execute=[@n1.NewEnumerator0, @n1.NewEnumerator1, @n1.NewEnumerator2]) {is=title K2Node_CallParentFunction}
   OverrideInputActions!(
    execute=@n1.NewEnumerator3
    self:WBP UIKit Tabs Base Object Reference
    Next Tab:Data Table Row Handle Structure=Get NextTabInputActionData() -> (NextTabInputActionData:Data Table Row Handle Structure)
    PreviousTab:Data Table Row Handle Structure=Get PreviousTabInputActionData() -> (PreviousTabInputActionData:Data Table Row Handle Structure)
   ) -> (then)
  ]}
]
variables [Input Actions Override {type=E_UI_TabsInputActionType, default=Custom}]
widgets {
 info {
  inheritedWidgetCount=0
  parentClass=/Game/UI/Frontend/TopBar/WBP_UIKit_Tabs_TopNavBar.WBP_UIKit_Tabs_TopNavBar_C
  rootWidgetClass=null
  widgetCount=0
 }
 widgets []
}
